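# Material function for the Central Picnic chest's center gem, adding an animated liquid whose level takes a Fill Alpha input.
# Sine-driven tilt shapes the surface; a virtual-plane projection supplies cap UVs, and cap, walls and foam are colored separately.
# Scrolling textures and clip-space, radial and glass-thickness masks shape detail and warping fed to MF_CP_ChestGems_CenterGem.
# Fresnel adds input-controlled edge glow, and liquid brightness is divided by eye adaptation.
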MaterialFunction /Game/Environments/Helios/Props/CentralPicnic/Materials/MaterialFunctions/MF_CP_ChestGems_CenterGemWithLiquid
graph {expressions {
  MaterialExpressionAbs_1 {class=MaterialExpressionAbs, inputs [None {from=MaterialExpressionNamedRerouteUsage_36, output=""}], properties {}}
  MaterialExpressionAdd_11 {
   class=MaterialExpressionAdd
   inputs [
    A {from=MaterialExpressionSubtract_26}
    B {from=MaterialExpressionSine_3}
   ]
  }
  MaterialExpressionAdd_12 {
   class=MaterialExpressionAdd
   inputs [
    A {from=""}
    B {from=MaterialExpressionMultiply_45}
   ]
   properties {constA=0.5}
  }
  MaterialExpressionAdd_14 {
   class=MaterialExpressionAdd
   inputs [
    A {from=MaterialExpressionMultiply_52}
    B {from=MaterialExpressionMultiply_46}
   ]
  }
  MaterialExpressionAdd_15 {
   class=MaterialExpressionAdd
   inputs [
    A {from=MaterialExpressionAdd_14}
    B {from=MaterialExpressionMaterialFunctionCall_20, output=Result}
   ]
  }
  MaterialExpressionAdd_16 {
   class=MaterialExpressionAdd
   inputs [
    A {from=MaterialExpressionAdd_7}
    B {from=""}
   ]
   properties {constB=-2}
  }
  MaterialExpressionAdd_17 {
   class=MaterialExpressionAdd
   inputs [
    A {from=MaterialExpressionMultiply_82}
    B {from=MaterialExpressionMultiply_57}
   ]
  }
  MaterialExpressionAdd_18 {
   class=MaterialExpressionAdd
   inputs [
    A {from=MaterialExpressionMultiply_55}
    B {from=MaterialExpressionMultiply_54}
   ]
  }
  MaterialExpressionAdd_19 {
   class=MaterialExpressionAdd
   inputs [
    A {from=MaterialExpressionTextureSample_5, output=RGB}
    B {from=MaterialExpressionTextureSample_6, output=RGB}
   ]
  }
  MaterialExpressionAdd_2 {
   class=MaterialExpressionAdd
   inputs [
    A {from=MaterialExpressionMaterialFunctionCall_0, output=Orb Center World Pos}
    B {from=MaterialExpressionWorldPosition_1, output=Z}
   ]
  }
  MaterialExpressionAdd_20 {
   class=MaterialExpressionAdd
   inputs [
    A {from=MaterialExpressionMultiply_63}
    B {from=MaterialExpressionLinearInterpolate_30}
   ]
  }
  MaterialExpressionAdd_21 {
   class=MaterialExpressionAdd
   inputs [
    A {from=MaterialExpressionVectorParameter_7, output=RGB}
    B {from=MaterialExpressionScalarParameter_35}
   ]
   properties {constB=0.04}
  }
  MaterialExpressionAdd_22 {
   class=MaterialExpressionAdd
   inputs [
    A {from=MaterialExpressionReroute_69}
    B {from=MaterialExpressionNamedRerouteDeclaration_7}
   ]
  }
  MaterialExpressionAdd_23 {
   class=MaterialExpressionAdd
   inputs [
    A {from=MaterialExpressionReroute_86}
    B {from=MaterialExpressionTransform_0}
   ]
  }
  MaterialExpressionAdd_24 {
   class=MaterialExpressionAdd
   inputs [
    A {from=MaterialExpressionReroute_85}
    B {from=MaterialExpressionMaterialFunctionCall_37, output=Result}
   ]
  }
  MaterialExpressionAdd_25 {
   class=MaterialExpressionAdd
   inputs [
    A {from=MaterialExpressionSubtract_14}
    B {from=MaterialExpressionScalarParameter_40}
   ]
  }
  MaterialExpressionAdd_28 {
   class=MaterialExpressionAdd
   inputs [
    A {from=MaterialExpressionAdd_24}
    B {from=MaterialExpressionMaterialFunctionCall_38, output=Result}
   ]
  }
  MaterialExpressionAdd_7 {
   class=MaterialExpressionAdd
   inputs [
    A {from=MaterialExpressionComponentMask_1}
    B {from=MaterialExpressionMultiply_16}
   ]
  }
  MaterialExpressionAppendVector_3 {
   class=MaterialExpressionAppendVector
   inputs [
    A {from=MaterialExpressionVectorParameter_12, output=R}
    B {from=MaterialExpressionVectorParameter_12, output=R}
   ]
  }
  MaterialExpressionAppendVector_4 {
   class=MaterialExpressionAppendVector
   inputs [
    A {from=MaterialExpressionVectorParameter_12, output=B}
    B {from=MaterialExpressionVectorParameter_12, output=B}
   ]
  }
  MaterialExpressionAppendVector_5 {
   class=MaterialExpressionAppendVector
   inputs [
    A {from=MaterialExpressionConstant_11, output=""}
    B {from=MaterialExpressionLinearInterpolate_29, output=""}
   ]
  }
  MaterialExpressionCameraPositionWS_0 {class=MaterialExpressionCameraPositionWS, inputs [], properties {}}
  MaterialExpressionCeil_0 {class=MaterialExpressionCeil, inputs [None {from=MaterialExpressionNamedRerouteUsage_1, output=""}], properties {}}
  MaterialExpressionCeil_1 {class=MaterialExpressionCeil, inputs [None {from=MaterialExpressionDotProduct_0, output=""}], properties {}}
  MaterialExpressionClamp_1 {
   class=MaterialExpressionClamp
   inputs [
    None {from=MaterialExpressionDivide_3, output=""}
    Min {from="", output=""}
    Max {from="", output=""}
   ]
   properties {}
  }
  MaterialExpressionComponentMask_0 {class=MaterialExpressionComponentMask, inputs [None {from=MaterialExpressionSubtract_2}], properties {b=true}}
  MaterialExpressionComponentMask_1 {class=MaterialExpressionComponentMask, inputs [None {from=MaterialExpressionSubtract_9}], properties {b=true}}
  MaterialExpressionComponentMask_2 {class=MaterialExpressionComponentMask, inputs [None {from=MaterialExpressionSubtract_9}], properties {g=true}}
  MaterialExpressionComponentMask_3 {class=MaterialExpressionComponentMask, inputs [None {from=MaterialExpressionMultiply_81}], properties {b=true}}
  MaterialExpressionComponentMask_4 {class=MaterialExpressionComponentMask, inputs [None {from=MaterialExpressionVectorParameter_10}], properties {b=true, g=true, r=true}}
  MaterialExpressionComponentMask_5 {class=MaterialExpressionComponentMask, inputs [None {from=MaterialExpressionFunctionInput_3}], properties {b=true, g=true, r=true}}
  MaterialExpressionComponentMask_7 {class=MaterialExpressionComponentMask, inputs [None {from=MaterialExpressionFunctionInput_3}], properties {a=true}}
  MaterialExpressionConstant2Vector_0 {class=MaterialExpressionConstant2Vector, properties {g=-1, r=1}}
  MaterialExpressionConstant3Vector_0 {class=MaterialExpressionConstant3Vector, properties {constant=<rgba 0 0.5 1 0>}}
  MaterialExpressionConstant3Vector_1 {class=MaterialExpressionConstant3Vector, properties {constant=<rgba 0 1 0 0>}}
  MaterialExpressionConstant3Vector_2 {class=MaterialExpressionConstant3Vector, properties {constant=<rgba 0 0 1 0>}}
  MaterialExpressionConstant3Vector_3 {class=MaterialExpressionConstant3Vector, properties {constant=<rgba 0 -0.5 1 0>}}
  MaterialExpressionConstant3Vector_4 {class=MaterialExpressionConstant3Vector, properties {constant=<rgba 0 0 1 0>}}
  MaterialExpressionConstant4Vector_0 {class=MaterialExpressionConstant4Vector, inputs [], properties {constant=<rgba 1 1 1 0>}}
  MaterialExpressionConstant_10 {class=MaterialExpressionConstant, properties {r=300}}
  MaterialExpressionConstant_11 {class=MaterialExpressionConstant, properties {r=1}}
  MaterialExpressionConstant_12 {class=MaterialExpressionConstant}
  MaterialExpressionConstant_13 {class=MaterialExpressionConstant, properties {r=1}}
  MaterialExpressionConstant_18 {class=MaterialExpressionConstant}
  MaterialExpressionConstant_19 {class=MaterialExpressionConstant, properties {r=1}}
  MaterialExpressionConstant_2 {class=MaterialExpressionConstant, properties {r=1}}
  MaterialExpressionConstant_20 {class=MaterialExpressionConstant}
  MaterialExpressionConstant_21 {class=MaterialExpressionConstant}
  MaterialExpressionConstant_22 {class=MaterialExpressionConstant, properties {r=0.2}}
  MaterialExpressionConstant_23 {class=MaterialExpressionConstant, properties {r=1}}
  MaterialExpressionConstant_4 {class=MaterialExpressionConstant}
  MaterialExpressionConstant_6 {class=MaterialExpressionConstant, properties {r=3}}
  MaterialExpressionCrossProduct_0 {
   class=MaterialExpressionCrossProduct
   inputs [
    A {from=MaterialExpressionNormalize_2, output=""}
    B {from=MaterialExpressionConstant3Vector_1, output=""}
   ]
   properties {}
  }
  MaterialExpressionDistance_1 {
   class=MaterialExpressionDistance
   inputs [
    A {from=MaterialExpressionSubtract_24, output=""}
    B {from=MaterialExpressionSubtract_1, output=""}
   ]
   properties {}
  }
  MaterialExpressionDivide_0 {
   class=MaterialExpressionDivide
   inputs [
    A {from=MaterialExpressionMultiply_58}
    B {from=MaterialExpressionEyeAdaptation_0, output=EyeAdaptation}
   ]
  }
  MaterialExpressionDivide_1 {
   class=MaterialExpressionDivide
   inputs [
    A {from=MaterialExpressionSubtract_17}
    B {from=MaterialExpressionScalarParameter_5}
   ]
  }
  MaterialExpressionDivide_2 {
   class=MaterialExpressionDivide
   inputs [
    A {from=MaterialExpressionDistance_1}
    B {from=MaterialExpressionScalarParameter_10}
   ]
   properties {constB=500}
  }
  MaterialExpressionDivide_3 {
   class=MaterialExpressionDivide
   inputs [
    A {from=MaterialExpressionSubtract_11}
    B {from=MaterialExpressionScalarParameter_19}
   ]
  }
  MaterialExpressionDivide_4 {
   class=MaterialExpressionDivide
   inputs [
    A {from=MaterialExpressionScalarParameter_25}
    B {from=""}
   ]
  }
  MaterialExpressionDivide_5 {
   class=MaterialExpressionDivide
   inputs [
    A {from=MaterialExpressionAdd_19}
    B {from=""}
   ]
  }
  MaterialExpressionDotProduct_0 {
   class=MaterialExpressionDotProduct
   inputs [
    A {from=MaterialExpressionConstant3Vector_2, output=""}
    B {from=MaterialExpressionNormalize_3, output=""}
   ]
   properties {}
  }
  MaterialExpressionEyeAdaptation_0 {class=MaterialExpressionEyeAdaptation, inputs [], properties {}}
  MaterialExpressionFunctionInput_0 {
   class=MaterialExpressionFunctionInput
   inputs [Preview {from=MaterialExpressionConstant_22, output=""}]
   properties {inputName=Fill Alpha, inputType=FunctionInput_Scalar, sortPriority=5}
  }
  MaterialExpressionFunctionInput_1 {
   class=MaterialExpressionFunctionInput
   inputs [Preview {from=MaterialExpressionMakeMaterialAttributes_2, output=""}]
   properties {bUsePreviewValueAsDefault=true, inputType=FunctionInput_MaterialAttributes}
  }
  MaterialExpressionFunctionInput_2 {
   class=MaterialExpressionFunctionInput
   inputs [Preview {from=MaterialExpressionConstant_23, output=""}]
   properties {bUsePreviewValueAsDefault=true, inputName=Center Gem Mask, inputType=FunctionInput_Scalar, sortPriority=3}
  }
  MaterialExpressionFunctionInput_3 {
   class=MaterialExpressionFunctionInput
   inputs [Preview {from=MaterialExpressionConstant4Vector_0, output=RGBA}]
   properties {inputName=Gem Edge Glow Color, inputType=FunctionInput_Vector4, sortPriority=7}
  }
  MaterialExpressionFunctionInput_4 {
   class=MaterialExpressionFunctionInput
   inputs [Preview {from=MaterialExpressionConstant_2, output=""}]
   properties {inputName=Gem Edge Glow Multi, inputType=FunctionInput_Scalar, sortPriority=7}
  }
  MaterialExpressionFunctionOutput_0 {class=MaterialExpressionFunctionOutput, inputs [None {from=MaterialExpressionSetMaterialAttributes_1, output=""}]}
  MaterialExpressionGetMaterialAttributes_0 {
   class=MaterialExpressionGetMaterialAttributes
   inputs [None {from=MaterialExpressionIf_3, output=""}]
   properties {attributeGetTypes [#1]}
  }
  MaterialExpressionGetMaterialAttributes_3 {
   class=MaterialExpressionGetMaterialAttributes
   inputs [None {from=MaterialExpressionReroute_70, output=""}]
   properties {attributeGetTypes [#2, #1]}
  }
  MaterialExpressionGetMaterialAttributes_5 {
   class=MaterialExpressionGetMaterialAttributes
   inputs [None {from=MaterialExpressionReroute_100, output=""}]
   properties {attributeGetTypes [#2, #1]}
  }
  MaterialExpressionGetMaterialAttributes_6 {
   class=MaterialExpressionGetMaterialAttributes
   inputs [None {from=MaterialExpressionSetMaterialAttributes_0, output=""}]
   properties {attributeGetTypes [#2, #1]}
  }
  MaterialExpressionGetMaterialAttributes_7 {
   class=MaterialExpressionGetMaterialAttributes
   inputs [None {from=MaterialExpressionMaterialFunctionCall_10, output=Material}]
   properties {attributeGetTypes [#2, #1]}
  }
  MaterialExpressionIf_2 {
   class=MaterialExpressionIf
   inputs [
    A {from=MaterialExpressionReroute_4, output=""}
    B {from="", output=""}
    "A > B" {from=MaterialExpressionSubtract_23, output=""}
    "A == B" {from=MaterialExpressionConstant_4, output=""}
    "A < B" {from=MaterialExpressionConstant_4, output=""}
   ]
   properties {}
  }
  MaterialExpressionIf_3 {
   class=MaterialExpressionIf
   inputs [
    A {from=MaterialExpressionNamedRerouteUsage_51, output=""}
    B {from="", output=""}
    "A > B" {from=MaterialExpressionReroute_89, output=""}
    "A == B" {from=MaterialExpressionReroute_87, output=""}
    "A < B" {from=MaterialExpressionReroute_87, output=""}
   ]
   properties {}
  }
  MaterialExpressionLinearInterpolate_0 {
   class=MaterialExpressionLinearInterpolate
   inputs [
    A {from=""}
    B {from=MaterialExpressionMax_8}
    Alpha {from=MaterialExpressionScalarParameter_0}
   ]
  }
  MaterialExpressionLinearInterpolate_1 {
   class=MaterialExpressionLinearInterpolate
   inputs [
    A {from=""}
    B {from=MaterialExpressionScalarParameter_6}
    Alpha {from=MaterialExpressionNamedRerouteUsage_0}
   ]
   properties {constA=1}
  }
  MaterialExpressionLinearInterpolate_17 {
   class=MaterialExpressionLinearInterpolate
   inputs [
    A {from=MaterialExpressionStaticSwitchParameter_0}
    B {from=MaterialExpressionStaticSwitchParameter_0}
    Alpha {from=MaterialExpressionReroute_20}
   ]
   properties {constA=1, desc=Cap}
  }
  MaterialExpressionLinearInterpolate_2 {
   class=MaterialExpressionLinearInterpolate
   inputs [
    A {from=MaterialExpressionMultiply_50}
    B {from=""}
    Alpha {from=MaterialExpressionReroute_5}
   ]
  }
  MaterialExpressionLinearInterpolate_20 {
   class=MaterialExpressionLinearInterpolate
   inputs [
    A {from=MaterialExpressionLinearInterpolate_17}
    B {from=MaterialExpressionReroute_59}
    Alpha {from=MaterialExpressionReroute_43}
   ]
   properties {desc=Walls}
  }
  MaterialExpressionLinearInterpolate_21 {
   class=MaterialExpressionLinearInterpolate
   inputs [
    A {from=MaterialExpressionLinearInterpolate_20}
    B {from=MaterialExpressionVectorParameter_8, output=RGB}
    Alpha {from=MaterialExpressionReroute_50}
   ]
   properties {desc=Foam}
  }
  MaterialExpressionLinearInterpolate_22 {
   class=MaterialExpressionLinearInterpolate
   inputs [
    A {from=MaterialExpressionAdd_21}
    B {from=MaterialExpressionVectorParameter_7, output=RGB}
    Alpha {from=MaterialExpressionNamedRerouteUsage_30}
   ]
  }
  MaterialExpressionLinearInterpolate_23 {
   class=MaterialExpressionLinearInterpolate
   inputs [
    A {from=MaterialExpressionLinearInterpolate_27}
    B {from=MaterialExpressionGetMaterialAttributes_5, output=BaseColor}
    Alpha {from=MaterialExpressionReroute_60}
   ]
  }
  MaterialExpressionLinearInterpolate_24 {
   class=MaterialExpressionLinearInterpolate
   inputs [
    A {from=MaterialExpressionLinearInterpolate_28}
    B {from=MaterialExpressionReroute_68}
    Alpha {from=MaterialExpressionReroute_60}
   ]
  }
  MaterialExpressionLinearInterpolate_26 {
   class=MaterialExpressionLinearInterpolate
   inputs [
    A {from=MaterialExpressionScalarParameter_2}
    B {from=MaterialExpressionScalarParameter_1}
    Alpha {from=MaterialExpressionDivide_5}
   ]
   properties {constA=0.7, constB=1.2}
  }
  MaterialExpressionLinearInterpolate_27 {
   class=MaterialExpressionLinearInterpolate
   inputs [
    A {from=MaterialExpressionGetMaterialAttributes_3, output=BaseColor}
    B {from=MaterialExpressionMultiply_65}
    Alpha {from=MaterialExpressionReroute_63}
   ]
  }
  MaterialExpressionLinearInterpolate_28 {
   class=MaterialExpressionLinearInterpolate
   inputs [
    A {from=MaterialExpressionGetMaterialAttributes_3, output=EmissiveColor}
    B {from=MaterialExpressionMultiply_75}
    Alpha {from=MaterialExpressionMultiply_60}
   ]
  }
  MaterialExpressionLinearInterpolate_29 {
   class=MaterialExpressionLinearInterpolate
   inputs [
    A {from=""}
    B {from=MaterialExpressionScalarParameter_30}
    Alpha {from=MaterialExpressionAbs_1}
   ]
   properties {constA=0.1, constB=5}
  }
  MaterialExpressionLinearInterpolate_3 {
   class=MaterialExpressionLinearInterpolate
   inputs [
    A {from=""}
    B {from=MaterialExpressionScalarParameter_7}
    Alpha {from=MaterialExpressionMaterialFunctionCall_0, output=Distance Mask}
   ]
   properties {constA=1}
  }
  MaterialExpressionLinearInterpolate_30 {
   class=MaterialExpressionLinearInterpolate
   inputs [
    A {from=MaterialExpressionNamedRerouteUsage_41}
    B {from=MaterialExpressionMultiply_64}
    Alpha {from=MaterialExpressionRound_0}
   ]
  }
  MaterialExpressionLinearInterpolate_31 {
   class=MaterialExpressionLinearInterpolate
   inputs [
    A {from=MaterialExpressionScalarParameter_29}
    B {from=""}
    Alpha {from=MaterialExpressionMaterialFunctionCall_32, output=Glass Thickness Mask}
   ]
  }
  MaterialExpressionLinearInterpolate_33 {
   class=MaterialExpressionLinearInterpolate
   inputs [
    A {from=MaterialExpressionReroute_81}
    B {from=MaterialExpressionLinearInterpolate_24}
    Alpha {from=MaterialExpressionScalarParameter_34}
   ]
  }
  MaterialExpressionLinearInterpolate_34 {
   class=MaterialExpressionLinearInterpolate
   inputs [
    A {from=MaterialExpressionReroute_80}
    B {from=MaterialExpressionLinearInterpolate_23}
    Alpha {from=MaterialExpressionScalarParameter_34}
   ]
  }
  MaterialExpressionLinearInterpolate_36 {
   class=MaterialExpressionLinearInterpolate
   inputs [
    A {from=MaterialExpressionGetMaterialAttributes_3, output=EmissiveColor}
    B {from=""}
    Alpha {from=MaterialExpressionSaturate_0}
   ]
   properties {constB=20}
  }
  MaterialExpressionLinearInterpolate_37 {
   class=MaterialExpressionLinearInterpolate
   inputs [
    A {from=MaterialExpressionGetMaterialAttributes_6, output=EmissiveColor}
    B {from=MaterialExpressionLinearInterpolate_42}
    Alpha {from=MaterialExpressionReroute_74}
   ]
  }
  MaterialExpressionLinearInterpolate_38 {
   class=MaterialExpressionLinearInterpolate
   inputs [
    A {from=MaterialExpressionGetMaterialAttributes_6, output=BaseColor}
    B {from=MaterialExpressionLinearInterpolate_43}
    Alpha {from=MaterialExpressionReroute_74}
   ]
  }
  MaterialExpressionLinearInterpolate_39 {
   class=MaterialExpressionLinearInterpolate
   inputs [
    A {from=MaterialExpressionReroute_82}
    B {from=MaterialExpressionMax_10}
    Alpha {from=MaterialExpressionNamedRerouteUsage_46}
   ]
  }
  MaterialExpressionLinearInterpolate_4 {
   class=MaterialExpressionLinearInterpolate
   inputs [
    A {from=MaterialExpressionMultiply_5}
    B {from=MaterialExpressionGetMaterialAttributes_0, output=EmissiveColor}
    Alpha {from=MaterialExpressionMaterialFunctionCall_1, output=Fresnel}
   ]
   properties {constA=1}
  }
  MaterialExpressionLinearInterpolate_40 {
   class=MaterialExpressionLinearInterpolate
   inputs [
    A {from=MaterialExpressionScalarParameter_28}
    B {from=MaterialExpressionMultiply_85}
    Alpha {from=MaterialExpressionNamedRerouteUsage_56}
   ]
  }
  MaterialExpressionLinearInterpolate_41 {
   class=MaterialExpressionLinearInterpolate
   inputs [
    A {from=MaterialExpressionScalarParameter_27}
    B {from=MaterialExpressionMultiply_86}
    Alpha {from=MaterialExpressionNamedRerouteUsage_56}
   ]
  }
  MaterialExpressionLinearInterpolate_42 {
   class=MaterialExpressionLinearInterpolate
   inputs [
    A {from=MaterialExpressionGetMaterialAttributes_6, output=EmissiveColor}
    B {from=MaterialExpressionVectorParameter_9, output=RGB}
    Alpha {from=MaterialExpressionVectorParameter_9, output=RGB}
   ]
  }
  MaterialExpressionLinearInterpolate_43 {
   class=MaterialExpressionLinearInterpolate
   inputs [
    A {from=MaterialExpressionGetMaterialAttributes_6, output=BaseColor}
    B {from=MaterialExpressionVectorParameter_9, output=RGB}
    Alpha {from=MaterialExpressionVectorParameter_9, output=RGB}
   ]
  }
  MaterialExpressionLinearInterpolate_5 {
   class=MaterialExpressionLinearInterpolate
   inputs [
    A {from=""}
    B {from=MaterialExpressionScalarParameter_8}
    Alpha {from=MaterialExpressionReroute_5}
   ]
   properties {constA=1}
  }
  MaterialExpressionLinearInterpolate_6 {
   class=MaterialExpressionLinearInterpolate
   inputs [
    A {from=""}
    B {from=MaterialExpressionMaterialFunctionCall_21, output=Fresnel}
    Alpha {from=MaterialExpressionNamedRerouteUsage_1}
   ]
   properties {constA=1}
  }
  MaterialExpressionLinearInterpolate_7 {
   class=MaterialExpressionLinearInterpolate
   inputs [
    A {from=MaterialExpressionLinearInterpolate_6}
    B {from=MaterialExpressionReroute_8}
    Alpha {from=MaterialExpressionMax_0}
   ]
   properties {constA=1}
  }
  MaterialExpressionLinearInterpolate_8 {
   class=MaterialExpressionLinearInterpolate
   inputs [
    A {from=MaterialExpressionReroute_7}
    B {from=MaterialExpressionLinearInterpolate_7}
    Alpha {from=MaterialExpressionCeil_0}
   ]
  }
  MaterialExpressionLinearInterpolate_9 {
   class=MaterialExpressionLinearInterpolate
   inputs [
    A {from=MaterialExpressionGetMaterialAttributes_0, output=EmissiveColor}
    B {from=MaterialExpressionComponentMask_5}
    Alpha {from=MaterialExpressionComponentMask_7}
   ]
  }
  MaterialExpressionMakeMaterialAttributes_2 {
   class=MaterialExpressionMakeMaterialAttributes
   inputs [
    BaseColor {from="", output=""}
    Metallic {from="", output=""}
    Specular {from="", output=""}
    Roughness {from="", output=""}
    Anisotropy {from="", output=""}
    EmissiveColor {from="", output=""}
    Opacity {from="", output=""}
    OpacityMask {from="", output=""}
    Normal {from="", output=""}
    Tangent {from="", output=""}
    WorldPositionOffset {from="", output=""}
    SubsurfaceColor {from="", output=""}
    ClearCoat {from="", output=""}
    ClearCoatRoughness {from="", output=""}
    AmbientOcclusion {from="", output=""}
    Refraction {from="", output=""}
    CustomizedUVs_0 {from="", output=""}
    CustomizedUVs_1 {from="", output=""}
    CustomizedUVs_2 {from="", output=""}
    CustomizedUVs_3 {from="", output=""}
    CustomizedUVs_4 {from="", output=""}
    CustomizedUVs_5 {from="", output=""}
    CustomizedUVs_6 {from="", output=""}
    CustomizedUVs_7 {from="", output=""}
    PixelDepthOffset {from="", output=""}
    ShadingModel {from="", output=""}
    Displacement {from="", output=""}
   ]
   properties {}
  }
  MaterialExpressionMaterialFunctionCall_0 {
   class=MaterialExpressionMaterialFunctionCall
   inputs [WarpMask {from=""}]
   properties {materialFunction=&/MF_CP_ChestGems_Utilities}
  }
  MaterialExpressionMaterialFunctionCall_1 {
   class=MaterialExpressionMaterialFunctionCall
   inputs [
    Brightness {from=MaterialExpressionScalarParameter_13}
    Exponent {from=MaterialExpressionScalarParameter_12}
    Camera Vector {from=""}
    Vertex / Pixel Normal {from=""}
   ]
   properties {materialFunction=&MF_Fresnel_FX}
  }
  MaterialExpressionMaterialFunctionCall_10 {
   class=MaterialExpressionMaterialFunctionCall
   inputs [
    Material {from=MaterialExpressionReroute_9}
    Mask {from=MaterialExpressionFunctionInput_2}
    WarpMask {from=MaterialExpressionMultiply_61}
    Constellation Wisp Mask Opacity {from=MaterialExpressionConstant_12}
   ]
   properties {materialFunction=&/MF_CP_ChestGems_CenterGem}
  }
  MaterialExpressionMaterialFunctionCall_13 {
   class=MaterialExpressionMaterialFunctionCall
   inputs [
    UVs {from=MaterialExpressionReroute_54}
    CenterPosition {from=MaterialExpressionReroute_52}
    Radius {from=MaterialExpressionMaterialFunctionCall_14, output=Result}
    Density {from=MaterialExpressionScalarParameter_45}
    Invert Density {from=""}
   ]
   properties {materialFunction=&RadialGradientExponential}
  }
  MaterialExpressionMaterialFunctionCall_14 {
   class=MaterialExpressionMaterialFunctionCall
   inputs [
    A {from=MaterialExpressionScalarParameter_37}
    B {from=MaterialExpressionScalarParameter_36}
    C {from=MaterialExpressionScalarParameter_37}
    Alpha {from=MaterialExpressionNamedRerouteUsage_53}
   ]
   properties {materialFunction=&Lerp_3Color}
  }
  MaterialExpressionMaterialFunctionCall_19 {
   class=MaterialExpressionMaterialFunctionCall
   inputs [
    A {from=MaterialExpressionConstant3Vector_3}
    B {from=MaterialExpressionConstant3Vector_4}
    C {from=MaterialExpressionConstant3Vector_0}
    Alpha {from=MaterialExpressionAdd_12}
   ]
   properties {materialFunction=&Lerp_3Color}
  }
  MaterialExpressionMaterialFunctionCall_20 {
   class=MaterialExpressionMaterialFunctionCall
   inputs [
    X {from=MaterialExpressionMultiply_49}
    Y {from=""}
   ]
   properties {materialFunction=&MakeFloat2}
  }
  MaterialExpressionMaterialFunctionCall_21 {
   class=MaterialExpressionMaterialFunctionCall
   inputs [
    Brightness {from=""}
    Exponent {from=MaterialExpressionScalarParameter_3}
    Camera Vector {from=""}
    Vertex / Pixel Normal {from=""}
   ]
   properties {materialFunction=&MF_Fresnel_FX}
  }
  MaterialExpressionMaterialFunctionCall_22 {
   class=MaterialExpressionMaterialFunctionCall
   inputs [
    UVs {from=MaterialExpressionReroute_78}
    CenterPosition {from=MaterialExpressionReroute_79}
    Radius {from=MaterialExpressionMultiply_71}
    Density {from=MaterialExpressionConstant_6}
    Invert Density {from=""}
   ]
   properties {materialFunction=&RadialGradientExponential}
  }
  MaterialExpressionMaterialFunctionCall_23 {
   class=MaterialExpressionMaterialFunctionCall
   inputs [
    Projection Pivot Position {from=MaterialExpressionAdd_24}
    Plane Depth Offset {from=MaterialExpressionConstant_21}
    "Coordinate Scale (WS)" {from=MaterialExpressionConstant_10}
    World Space Projection Normal {from=MaterialExpressionNormalize_2}
    World Space Projection Tangent {from=MaterialExpressionNormalize_1}
    World Space Projection Binormal {from=MaterialExpressionNormalize_0}
   ]
   properties {
    desc="Cap Detail UVs, same as above but scaled to a different size"
    materialFunction=/Engine/Functions/Engine_MaterialFunctions02/UVs/ObjectAlignedVirtualPlaneCoordinates
   }
  }
  MaterialExpressionMaterialFunctionCall_24 {
   class=MaterialExpressionMaterialFunctionCall
   inputs [WarpMask {from=""}]
   properties {materialFunction=&MF_CP_ChestGems_Utilities}
  }
  MaterialExpressionMaterialFunctionCall_25 {
   class=MaterialExpressionMaterialFunctionCall
   inputs [
    World Position {from=MaterialExpressionAdd_28}
    Buffer UV Pixel Offset {from=""}
   ]
   properties {materialFunction=&TransformToClipSpace}
  }
  MaterialExpressionMaterialFunctionCall_26 {
   class=MaterialExpressionMaterialFunctionCall
   inputs [
    UVs {from=MaterialExpressionSubtract_31}
    Texture Scale {from=MaterialExpressionMultiply_70}
   ]
   properties {materialFunction=&ScaleUVsByCenter}
  }
  MaterialExpressionMaterialFunctionCall_27 {
   class=MaterialExpressionMaterialFunctionCall
   inputs [UV Channel {from=MaterialExpressionMaterialFunctionCall_26, output=UVs}]
   properties {materialFunction=&LinearGradient}
  }
  MaterialExpressionMaterialFunctionCall_31 {
   class=MaterialExpressionMaterialFunctionCall
   inputs [
    X {from=MaterialExpressionNamedRerouteUsage_50}
    Y {from=MaterialExpressionReroute_73}
    Z {from=MaterialExpressionSaturate_24}
   ]
   properties {materialFunction=&MakeFloat3}
  }
  MaterialExpressionMaterialFunctionCall_32 {
   class=MaterialExpressionMaterialFunctionCall
   inputs [WarpMask {from=""}]
   properties {materialFunction=&MF_CP_ChestGems_Utilities}
  }
  MaterialExpressionMaterialFunctionCall_33 {
   class=MaterialExpressionMaterialFunctionCall
   inputs [
    UVs {from=MaterialExpressionReroute_78}
    CenterPosition {from=MaterialExpressionReroute_79}
    Radius {from=MaterialExpressionMultiply_77}
    Density {from=MaterialExpressionScalarParameter_41}
    Invert Density {from=""}
   ]
   properties {materialFunction=&RadialGradientExponential}
  }
  MaterialExpressionMaterialFunctionCall_34 {
   class=MaterialExpressionMaterialFunctionCall
   inputs [
    A {from=MaterialExpressionConstant_18}
    B {from=MaterialExpressionConstant_13}
    C {from=MaterialExpressionConstant_18}
    Alpha {from=MaterialExpressionNamedRerouteUsage_54}
   ]
   properties {materialFunction=&Lerp_3Color}
  }
  MaterialExpressionMaterialFunctionCall_35 {
   class=MaterialExpressionMaterialFunctionCall
   inputs [
    A {from=MaterialExpressionConstant_20}
    B {from=MaterialExpressionConstant_20}
    C {from=MaterialExpressionConstant_19}
    Alpha {from=MaterialExpressionNamedRerouteUsage_55}
   ]
   properties {materialFunction=&Lerp_3Color}
  }
  MaterialExpressionMaterialFunctionCall_36 {
   class=MaterialExpressionMaterialFunctionCall
   inputs [WarpMask {from=""}]
   properties {materialFunction=&MF_CP_ChestGems_Utilities}
  }
  MaterialExpressionMaterialFunctionCall_37 {
   class=MaterialExpressionMaterialFunctionCall
   inputs [
    X {from=""}
    Y {from=""}
    Z {from=MaterialExpressionAdd_25}
   ]
   properties {materialFunction=&MakeFloat3}
  }
  MaterialExpressionMaterialFunctionCall_38 {
   class=MaterialExpressionMaterialFunctionCall
   inputs [
    X {from=""}
    Y {from=""}
    Z {from=MaterialExpressionScalarParameter_42}
   ]
   properties {materialFunction=&MakeFloat3}
  }
  MaterialExpressionMaterialFunctionCall_39 {
   class=MaterialExpressionMaterialFunctionCall
   inputs [
    A {from=MaterialExpressionReroute_102}
    B {from=MaterialExpressionReroute_102}
    C {from=MaterialExpressionMultiply_89}
    Alpha {from=MaterialExpressionNamedRerouteUsage_58}
   ]
   properties {materialFunction=&Lerp_3Color}
  }
  MaterialExpressionMaterialFunctionCall_40 {
   class=MaterialExpressionMaterialFunctionCall
   inputs [
    A {from=MaterialExpressionReroute_103}
    B {from=MaterialExpressionReroute_103}
    C {from=MaterialExpressionMultiply_88}
    Alpha {from=MaterialExpressionNamedRerouteUsage_59}
   ]
   properties {materialFunction=&Lerp_3Color}
  }
  MaterialExpressionMax_0 {
   class=MaterialExpressionMax
   inputs [
    A {from=MaterialExpressionSaturate_27}
    B {from=MaterialExpressionNamedRerouteUsage_4}
   ]
  }
  MaterialExpressionMax_10 {
   class=MaterialExpressionMax
   inputs [
    A {from=MaterialExpressionNamedRerouteUsage_48}
    B {from=MaterialExpressionMultiply_80}
   ]
  }
  MaterialExpressionMax_11 {
   class=MaterialExpressionMax
   inputs [
    A {from=MaterialExpressionReroute_43}
    B {from=MaterialExpressionReroute_101}
   ]
  }
  MaterialExpressionMax_6 {
   class=MaterialExpressionMax
   inputs [
    A {from=MaterialExpressionMax_11}
    B {from=MaterialExpressionReroute_35}
   ]
  }
  MaterialExpressionMax_7 {
   class=MaterialExpressionMax
   inputs [
    A {from=MaterialExpressionSaturate_26}
    B {from=MaterialExpressionNamedRerouteUsage_33}
   ]
  }
  MaterialExpressionMax_8 {
   class=MaterialExpressionMax
   inputs [
    A {from=MaterialExpressionSubtract_28}
    B {from=MaterialExpressionLinearInterpolate_39}
   ]
  }
  MaterialExpressionMax_9 {
   class=MaterialExpressionMax
   inputs [
    A {from=MaterialExpressionSaturate_25}
    B {from=MaterialExpressionMultiply_73}
   ]
  }
  MaterialExpressionMultiply_0 {
   class=MaterialExpressionMultiply
   inputs [
    A {from=MaterialExpressionLinearInterpolate_1}
    B {from=MaterialExpressionScalarParameter_4}
   ]
  }
  MaterialExpressionMultiply_1 {
   class=MaterialExpressionMultiply
   inputs [
    A {from=MaterialExpressionScalarParameter_26}
    B {from=MaterialExpressionLinearInterpolate_3}
   ]
  }
  MaterialExpressionMultiply_16 {
   class=MaterialExpressionMultiply
   inputs [
    A {from=MaterialExpressionComponentMask_2}
    B {from=MaterialExpressionMultiply_37}
   ]
  }
  MaterialExpressionMultiply_2 {
   class=MaterialExpressionMultiply
   inputs [
    A {from=MaterialExpressionOneMinus_4}
    B {from=MaterialExpressionPower_1}
   ]
  }
  MaterialExpressionMultiply_37 {
   class=MaterialExpressionMultiply
   inputs [
    A {from=MaterialExpressionSine_0}
    B {from=MaterialExpressionScalarParameter_18}
   ]
  }
  MaterialExpressionMultiply_39 {
   class=MaterialExpressionMultiply
   inputs [
    A {from=MaterialExpressionMultiply_81}
    B {from=""}
   ]
   properties {constB=2}
  }
  MaterialExpressionMultiply_4 {
   class=MaterialExpressionMultiply
   inputs [
    A {from=MaterialExpressionLinearInterpolate_21}
    B {from=MaterialExpressionLinearInterpolate_5}
   ]
  }
  MaterialExpressionMultiply_40 {
   class=MaterialExpressionMultiply
   inputs [
    A {from=MaterialExpressionNamedRerouteUsage_52}
    B {from=MaterialExpressionMultiply_39}
   ]
   properties {constB=2}
  }
  MaterialExpressionMultiply_43 {
   class=MaterialExpressionMultiply
   inputs [
    A {from=MaterialExpressionMaterialFunctionCall_13, output=RadialGradientExponential}
    B {from=MaterialExpressionCeil_1}
   ]
  }
  MaterialExpressionMultiply_44 {
   class=MaterialExpressionMultiply
   inputs [
    A {from=MaterialExpressionScalarParameter_23}
    B {from=""}
   ]
   properties {constB=100}
  }
  MaterialExpressionMultiply_45 {
   class=MaterialExpressionMultiply
   inputs [
    A {from=MaterialExpressionAdd_11}
    B {from=MaterialExpressionReroute_51}
   ]
  }
  MaterialExpressionMultiply_46 {
   class=MaterialExpressionMultiply
   inputs [
    A {from=MaterialExpressionTime_2}
    B {from=MaterialExpressionLinearInterpolate_41}
   ]
   properties {constB=0.1}
  }
  MaterialExpressionMultiply_49 {
   class=MaterialExpressionMultiply
   inputs [
    A {from=MaterialExpressionLinearInterpolate_41}
    B {from=""}
   ]
   properties {constB=0.33}
  }
  MaterialExpressionMultiply_5 {
   class=MaterialExpressionMultiply
   inputs [
    A {from=MaterialExpressionLinearInterpolate_9}
    B {from=MaterialExpressionFunctionInput_4}
   ]
  }
  MaterialExpressionMultiply_50 {
   class=MaterialExpressionMultiply
   inputs [
    A {from=MaterialExpressionNamedRerouteDeclaration_8}
    B {from=MaterialExpressionLinearInterpolate_8}
   ]
  }
  MaterialExpressionMultiply_51 {
   class=MaterialExpressionMultiply
   inputs [
    A {from=MaterialExpressionMultiply_4}
    B {from=MaterialExpressionLinearInterpolate_2}
   ]
  }
  MaterialExpressionMultiply_52 {
   class=MaterialExpressionMultiply
   inputs [
    A {from=MaterialExpressionTextureCoordinate_0}
    B {from=MaterialExpressionLinearInterpolate_40}
   ]
  }
  MaterialExpressionMultiply_53 {
   class=MaterialExpressionMultiply
   inputs [
    A {from=MaterialExpressionTime_4}
    B {from=""}
   ]
   properties {constB=0.1}
  }
  MaterialExpressionMultiply_54 {
   class=MaterialExpressionMultiply
   inputs [
    A {from=MaterialExpressionNamedRerouteUsage_49}
    B {from=MaterialExpressionScalarParameter_39}
   ]
   properties {constB=0.66}
  }
  MaterialExpressionMultiply_55 {
   class=MaterialExpressionMultiply
   inputs [
    A {from=MaterialExpressionMultiply_53}
    B {from=MaterialExpressionAppendVector_4}
   ]
   properties {constB=-0.45}
  }
  MaterialExpressionMultiply_56 {
   class=MaterialExpressionMultiply
   inputs [
    A {from=MaterialExpressionReroute_99}
    B {from=MaterialExpressionLinearInterpolate_22}
   ]
  }
  MaterialExpressionMultiply_57 {
   class=MaterialExpressionMultiply
   inputs [
    A {from=MaterialExpressionMultiply_53}
    B {from=MaterialExpressionAppendVector_3}
   ]
  }
  MaterialExpressionMultiply_58 {
   class=MaterialExpressionMultiply
   inputs [
    A {from=MaterialExpressionMultiply_51}
    B {from=MaterialExpressionMultiply_0}
   ]
   properties {constB=3}
  }
  MaterialExpressionMultiply_59 {
   class=MaterialExpressionMultiply
   inputs [
    A {from=MaterialExpressionSaturate_18}
    B {from=MaterialExpressionScalarParameter_33}
   ]
   properties {constB=5}
  }
  MaterialExpressionMultiply_60 {
   class=MaterialExpressionMultiply
   inputs [
    A {from=MaterialExpressionLinearInterpolate_31}
    B {from=MaterialExpressionReroute_63}
   ]
   properties {constB=0.5}
  }
  MaterialExpressionMultiply_61 {
   class=MaterialExpressionMultiply
   inputs [
    A {from=MaterialExpressionMaterialFunctionCall_31, output=Result}
    B {from=MaterialExpressionMultiply_62}
   ]
  }
  MaterialExpressionMultiply_62 {
   class=MaterialExpressionMultiply
   inputs [
    A {from=MaterialExpressionPower_0}
    B {from=MaterialExpressionScalarParameter_32}
   ]
   properties {constB=0.05}
  }
  MaterialExpressionMultiply_63 {
   class=MaterialExpressionMultiply
   inputs [
    A {from=MaterialExpressionTime_5}
    B {from=""}
   ]
   properties {constB=0.1}
  }
  MaterialExpressionMultiply_64 {
   class=MaterialExpressionMultiply
   inputs [
    A {from=MaterialExpressionTextureCoordinate_3}
    B {from=""}
   ]
   properties {constB=3}
  }
  MaterialExpressionMultiply_65 {
   class=MaterialExpressionMultiply
   inputs [
    A {from=MaterialExpressionGetMaterialAttributes_3, output=BaseColor}
    B {from=MaterialExpressionNamedRerouteUsage_44}
   ]
  }
  MaterialExpressionMultiply_66 {
   class=MaterialExpressionMultiply
   inputs [
    A {from=MaterialExpressionLinearInterpolate_36}
    B {from=MaterialExpressionNamedRerouteUsage_44}
   ]
  }
  MaterialExpressionMultiply_67 {
   class=MaterialExpressionMultiply
   inputs [
    A {from=MaterialExpressionMaterialFunctionCall_25, output=Clip Space XY}
    B {from=MaterialExpressionConstant2Vector_0}
   ]
  }
  MaterialExpressionMultiply_68 {
   class=MaterialExpressionMultiply
   inputs [
    A {from=MaterialExpressionMultiply_67}
    B {from=""}
   ]
   properties {constB=0.5}
  }
  MaterialExpressionMultiply_69 {
   class=MaterialExpressionMultiply
   inputs [
    A {from=MaterialExpressionReroute_66}
    B {from=MaterialExpressionAppendVector_5}
   ]
   properties {constB=1.1}
  }
  MaterialExpressionMultiply_70 {
   class=MaterialExpressionMultiply
   inputs [
    A {from=MaterialExpressionMultiply_69}
    B {from=MaterialExpressionReroute_65}
   ]
  }
  MaterialExpressionMultiply_71 {
   class=MaterialExpressionMultiply
   inputs [
    A {from=MaterialExpressionReroute_55}
    B {from=""}
   ]
   properties {constB=0.7}
  }
  MaterialExpressionMultiply_72 {
   class=MaterialExpressionMultiply
   inputs [
    A {from=MaterialExpressionOneMinus_1}
    B {from=""}
   ]
   properties {constB=2}
  }
  MaterialExpressionMultiply_73 {
   class=MaterialExpressionMultiply
   inputs [
    A {from=MaterialExpressionNamedRerouteUsage_38}
    B {from=""}
   ]
  }
  MaterialExpressionMultiply_75 {
   class=MaterialExpressionMultiply
   inputs [
    A {from=MaterialExpressionMultiply_66}
    B {from=""}
   ]
   properties {constB=30}
  }
  MaterialExpressionMultiply_76 {
   class=MaterialExpressionMultiply
   inputs [
    A {from=MaterialExpressionNamedRerouteDeclaration_19}
    B {from=MaterialExpressionNamedRerouteUsage_42}
   ]
  }
  MaterialExpressionMultiply_77 {
   class=MaterialExpressionMultiply
   inputs [
    A {from=MaterialExpressionReroute_55}
    B {from=MaterialExpressionScalarParameter_46}
   ]
   properties {constB=0.9}
  }
  MaterialExpressionMultiply_78 {
   class=MaterialExpressionMultiply
   inputs [
    A {from=MaterialExpressionMultiply_79}
    B {from=MaterialExpressionReroute_77}
   ]
  }
  MaterialExpressionMultiply_79 {
   class=MaterialExpressionMultiply
   inputs [
    A {from=MaterialExpressionMaterialFunctionCall_34, output=Result}
    B {from=MaterialExpressionNamedRerouteUsage_45}
   ]
  }
  MaterialExpressionMultiply_8 {
   class=MaterialExpressionMultiply
   inputs [
    A {from=MaterialExpressionReroute_1}
    B {from=MaterialExpressionReroute_99}
   ]
  }
  MaterialExpressionMultiply_80 {
   class=MaterialExpressionMultiply
   inputs [
    A {from=MaterialExpressionNamedRerouteUsage_47}
    B {from=MaterialExpressionMaterialFunctionCall_35, output=Result}
   ]
  }
  MaterialExpressionMultiply_81 {
   class=MaterialExpressionMultiply
   inputs [
    A {from=MaterialExpressionComponentMask_4}
    B {from=MaterialExpressionVectorParameter_10, output=Multiplier}
   ]
  }
  MaterialExpressionMultiply_82 {
   class=MaterialExpressionMultiply
   inputs [
    A {from=MaterialExpressionNamedRerouteUsage_49}
    B {from=MaterialExpressionScalarParameter_38}
   ]
  }
  MaterialExpressionMultiply_83 {
   class=MaterialExpressionMultiply
   inputs [
    A {from=MaterialExpressionNamedRerouteUsage_32}
    B {from=""}
   ]
   properties {constB=0.5}
  }
  MaterialExpressionMultiply_84 {
   class=MaterialExpressionMultiply
   inputs [
    A {from=MaterialExpressionMultiply_2}
    B {from=MaterialExpressionScalarParameter_44}
   ]
  }
  MaterialExpressionMultiply_85 {
   class=MaterialExpressionMultiply
   inputs [
    A {from=MaterialExpressionScalarParameter_28}
    B {from=""}
   ]
   properties {constB=2}
  }
  MaterialExpressionMultiply_86 {
   class=MaterialExpressionMultiply
   inputs [
    A {from=MaterialExpressionScalarParameter_27}
    B {from=""}
   ]
   properties {constB=2}
  }
  MaterialExpressionMultiply_87 {
   class=MaterialExpressionMultiply
   inputs [
    A {from=MaterialExpressionMaterialFunctionCall_22, output=RadialGradientExponential}
    B {from=MaterialExpressionNamedRerouteUsage_57}
   ]
   properties {constB=0.5}
  }
  MaterialExpressionMultiply_88 {
   class=MaterialExpressionMultiply
   inputs [
    A {from=MaterialExpressionGetMaterialAttributes_7, output=BaseColor}
    B {from=MaterialExpressionVectorParameter_13, output=RGB}
   ]
  }
  MaterialExpressionMultiply_89 {
   class=MaterialExpressionMultiply
   inputs [
    A {from=MaterialExpressionGetMaterialAttributes_7, output=EmissiveColor}
    B {from=MaterialExpressionVectorParameter_13, output=RGB}
   ]
  }
  MaterialExpressionNamedRerouteDeclaration_12 {
   class=MaterialExpressionNamedRerouteDeclaration
   inputs [None {from=MaterialExpressionMultiply_43}]
   properties {name=Mask_Liquid_Cap, nodeColor=<rgba 0.5 0 1 1>}
  }
  MaterialExpressionNamedRerouteDeclaration_14 {
   class=MaterialExpressionNamedRerouteDeclaration
   inputs [None {from=MaterialExpressionTextureSample_4, output=R}]
   properties {name=Wave Time, nodeColor=<rgba 0 1 0.4141 1>}
  }
  MaterialExpressionNamedRerouteDeclaration_15 {
   class=MaterialExpressionNamedRerouteDeclaration
   inputs [None {from=MaterialExpressionDotProduct_0}]
   properties {name=IsParallelToCapDeltaMask, nodeColor=<rgba 0 1 0.4375 1>}
  }
  MaterialExpressionNamedRerouteDeclaration_16 {
   class=MaterialExpressionNamedRerouteDeclaration
   inputs [None {from=MaterialExpressionMaterialFunctionCall_10, output=ConstellationWispsMask}]
   properties {name=ConstellationMask, nodeColor=<rgba 0.0781 0 1 1>}
  }
  MaterialExpressionNamedRerouteDeclaration_17 {
   class=MaterialExpressionNamedRerouteDeclaration
   inputs [None {from=MaterialExpressionMaterialFunctionCall_23, output=Center Aligned UV Plane}]
   properties {name=Cap UVs, nodeColor=<rgba 0.9922 1 0 1>}
  }
  MaterialExpressionNamedRerouteDeclaration_18 {
   class=MaterialExpressionNamedRerouteDeclaration
   inputs [None {from=MaterialExpressionVectorParameter_7, output=RGB}]
   properties {name=Liquid Color Tint, nodeColor=<rgba 0 1 0.4609 1>}
  }
  MaterialExpressionNamedRerouteDeclaration_19 {
   class=MaterialExpressionNamedRerouteDeclaration
   inputs [None {from=MaterialExpressionLinearInterpolate_39}]
   properties {name=Warp Mask, nodeColor=<rgba 0 1 0.9766 1>}
  }
  MaterialExpressionNamedRerouteDeclaration_20 {
   class=MaterialExpressionNamedRerouteDeclaration
   inputs [None {from=MaterialExpressionSaturate_1}]
   properties {name=FillValue, nodeColor=<rgba 1 0 0.4688 1>}
  }
  MaterialExpressionNamedRerouteDeclaration_7 {
   class=MaterialExpressionNamedRerouteDeclaration
   inputs [None {from=MaterialExpressionSaturate_22}]
   properties {name=DepthCenterWaterLine, nodeColor=<rgba 0 0.4844 1 1>}
  }
  MaterialExpressionNamedRerouteDeclaration_8 {
   class=MaterialExpressionNamedRerouteDeclaration
   inputs [None {from=MaterialExpressionSaturate_15}]
   properties {name=Mask_Liquid, nodeColor=<rgba 0.5938 0 1 1>}
  }
  MaterialExpressionNamedRerouteDeclaration_9 {
   class=MaterialExpressionNamedRerouteDeclaration
   inputs [None {from=MaterialExpressionOneMinus_0}]
   properties {name=Mask_Liquid_Walls, nodeColor=<rgba 0 0.6484 1 1>}
  }
  MaterialExpressionNamedRerouteUsage_0 {class=MaterialExpressionNamedRerouteUsage}
  MaterialExpressionNamedRerouteUsage_1 {class=MaterialExpressionNamedRerouteUsage}
  MaterialExpressionNamedRerouteUsage_27 {class=MaterialExpressionNamedRerouteUsage}
  MaterialExpressionNamedRerouteUsage_28 {class=MaterialExpressionNamedRerouteUsage}
  MaterialExpressionNamedRerouteUsage_30 {class=MaterialExpressionNamedRerouteUsage}
  MaterialExpressionNamedRerouteUsage_31 {class=MaterialExpressionNamedRerouteUsage}
  MaterialExpressionNamedRerouteUsage_32 {class=MaterialExpressionNamedRerouteUsage}
  MaterialExpressionNamedRerouteUsage_33 {class=MaterialExpressionNamedRerouteUsage}
  MaterialExpressionNamedRerouteUsage_34 {class=MaterialExpressionNamedRerouteUsage}
  MaterialExpressionNamedRerouteUsage_35 {class=MaterialExpressionNamedRerouteUsage}
  MaterialExpressionNamedRerouteUsage_36 {class=MaterialExpressionNamedRerouteUsage}
  MaterialExpressionNamedRerouteUsage_38 {class=MaterialExpressionNamedRerouteUsage}
  MaterialExpressionNamedRerouteUsage_4 {class=MaterialExpressionNamedRerouteUsage}
  MaterialExpressionNamedRerouteUsage_40 {class=MaterialExpressionNamedRerouteUsage}
  MaterialExpressionNamedRerouteUsage_41 {class=MaterialExpressionNamedRerouteUsage}
  MaterialExpressionNamedRerouteUsage_42 {class=MaterialExpressionNamedRerouteUsage}
  MaterialExpressionNamedRerouteUsage_43 {class=MaterialExpressionNamedRerouteUsage}
  MaterialExpressionNamedRerouteUsage_44 {class=MaterialExpressionNamedRerouteUsage}
  MaterialExpressionNamedRerouteUsage_45 {class=MaterialExpressionNamedRerouteUsage}
  MaterialExpressionNamedRerouteUsage_46 {class=MaterialExpressionNamedRerouteUsage}
  MaterialExpressionNamedRerouteUsage_47 {class=MaterialExpressionNamedRerouteUsage}
  MaterialExpressionNamedRerouteUsage_48 {class=MaterialExpressionNamedRerouteUsage}
  MaterialExpressionNamedRerouteUsage_49 {class=MaterialExpressionNamedRerouteUsage}
  MaterialExpressionNamedRerouteUsage_50 {class=MaterialExpressionNamedRerouteUsage}
  MaterialExpressionNamedRerouteUsage_51 {class=MaterialExpressionNamedRerouteUsage}
  MaterialExpressionNamedRerouteUsage_52 {class=MaterialExpressionNamedRerouteUsage}
  MaterialExpressionNamedRerouteUsage_53 {class=MaterialExpressionNamedRerouteUsage}
  MaterialExpressionNamedRerouteUsage_54 {class=MaterialExpressionNamedRerouteUsage}
  MaterialExpressionNamedRerouteUsage_55 {class=MaterialExpressionNamedRerouteUsage}
  MaterialExpressionNamedRerouteUsage_56 {class=MaterialExpressionNamedRerouteUsage}
  MaterialExpressionNamedRerouteUsage_57 {class=MaterialExpressionNamedRerouteUsage}
  MaterialExpressionNamedRerouteUsage_58 {class=MaterialExpressionNamedRerouteUsage}
  MaterialExpressionNamedRerouteUsage_59 {class=MaterialExpressionNamedRerouteUsage}
  MaterialExpressionNormalize_0 {class=MaterialExpressionNormalize, inputs [VectorInput {from=MaterialExpressionConstant3Vector_1, output=""}]}
  MaterialExpressionNormalize_1 {class=MaterialExpressionNormalize, inputs [VectorInput {from=MaterialExpressionCrossProduct_0, output=""}]}
  MaterialExpressionNormalize_2 {class=MaterialExpressionNormalize, inputs [VectorInput {from=MaterialExpressionMaterialFunctionCall_19, output=Result}]}
  MaterialExpressionNormalize_3 {class=MaterialExpressionNormalize, inputs [VectorInput {from=MaterialExpressionSubtract_13, output=""}]}
  MaterialExpressionOneMinus_0 {class=MaterialExpressionOneMinus, inputs [None {from=MaterialExpressionClamp_1}]}
  MaterialExpressionOneMinus_1 {class=MaterialExpressionOneMinus, inputs [None {from=MaterialExpressionSaturate_21}]}
  MaterialExpressionOneMinus_2 {class=MaterialExpressionOneMinus, inputs [None {from=MaterialExpressionClamp_0}]}
  MaterialExpressionOneMinus_4 {class=MaterialExpressionOneMinus, inputs [None {from=MaterialExpressionSaturate_4}]}
  MaterialExpressionPower_0 {
   class=MaterialExpressionPower
   inputs [
    Base {from=MaterialExpressionTextureSample_7, output=R}
    Exp {from=""}
   ]
  }
  MaterialExpressionPower_1 {
   class=MaterialExpressionPower
   inputs [
    Base {from=MaterialExpressionOneMinus_2}
    Exp {from=MaterialExpressionScalarParameter_11}
   ]
   properties {constExponent=50}
  }
  MaterialExpressionReroute_0 {class=MaterialExpressionReroute, inputs [None {from=MaterialExpressionVectorParameter_8, output=RGB}]}
  MaterialExpressionReroute_1 {class=MaterialExpressionReroute, inputs [None {from=MaterialExpressionReroute_0}]}
  MaterialExpressionReroute_10 {class=MaterialExpressionReroute, inputs [None {from=MaterialExpressionLinearInterpolate_21}]}
  MaterialExpressionReroute_100 {class=MaterialExpressionReroute, inputs [None {from=MaterialExpressionSetMaterialAttributes_4}]}
  MaterialExpressionReroute_101 {class=MaterialExpressionReroute, inputs [None {from=MaterialExpressionSaturate_27}]}
  MaterialExpressionReroute_102 {class=MaterialExpressionReroute, inputs [None {from=MaterialExpressionGetMaterialAttributes_7, output=EmissiveColor}]}
  MaterialExpressionReroute_103 {class=MaterialExpressionReroute, inputs [None {from=MaterialExpressionGetMaterialAttributes_7, output=BaseColor}]}
  MaterialExpressionReroute_18 {class=MaterialExpressionReroute, inputs [None {from=MaterialExpressionSubtract_10}]}
  MaterialExpressionReroute_20 {class=MaterialExpressionReroute, inputs [None {from=MaterialExpressionReroute_34}]}
  MaterialExpressionReroute_3 {class=MaterialExpressionReroute, inputs [None {from=MaterialExpressionMultiply_1}]}
  MaterialExpressionReroute_34 {class=MaterialExpressionReroute, inputs [None {from=MaterialExpressionReroute_35}]}
  MaterialExpressionReroute_35 {class=MaterialExpressionReroute, inputs [None {from=MaterialExpressionNamedRerouteDeclaration_12}]}
  MaterialExpressionReroute_4 {class=MaterialExpressionReroute, inputs [None {from=MaterialExpressionMultiply_1}]}
  MaterialExpressionReroute_43 {class=MaterialExpressionReroute, inputs [None {from=MaterialExpressionReroute_45}]}
  MaterialExpressionReroute_45 {class=MaterialExpressionReroute, inputs [None {from=MaterialExpressionNamedRerouteDeclaration_9}]}
  MaterialExpressionReroute_46 {class=MaterialExpressionReroute, inputs [None {from=MaterialExpressionSubtract_27}]}
  MaterialExpressionReroute_5 {class=MaterialExpressionReroute, inputs [None {from=MaterialExpressionReroute_101}]}
  MaterialExpressionReroute_50 {class=MaterialExpressionReroute, inputs [None {from=MaterialExpressionReroute_5}]}
  MaterialExpressionReroute_51 {class=MaterialExpressionReroute, inputs [None {from=MaterialExpressionScalarParameter_25}]}
  MaterialExpressionReroute_52 {class=MaterialExpressionReroute, inputs [None {from=MaterialExpressionReroute_91}]}
  MaterialExpressionReroute_53 {class=MaterialExpressionReroute, inputs [None {from=MaterialExpressionAdd_12}]}
  MaterialExpressionReroute_54 {class=MaterialExpressionReroute, inputs [None {from=MaterialExpressionNamedRerouteDeclaration_17}]}
  MaterialExpressionReroute_55 {class=MaterialExpressionReroute, inputs [None {from=MaterialExpressionMaterialFunctionCall_14, output=Result}]}
  MaterialExpressionReroute_56 {class=MaterialExpressionReroute, inputs [None {from=MaterialExpressionReroute_52}]}
  MaterialExpressionReroute_57 {class=MaterialExpressionReroute, inputs [None {from=MaterialExpressionReroute_54}]}
  MaterialExpressionReroute_58 {class=MaterialExpressionReroute, inputs [None {from=MaterialExpressionMultiply_50}]}
  MaterialExpressionReroute_59 {class=MaterialExpressionReroute, inputs [None {from=MaterialExpressionNamedRerouteDeclaration_18}]}
  MaterialExpressionReroute_6 {class=MaterialExpressionReroute, inputs [None {from=MaterialExpressionMaterialFunctionCall_21, output=Fresnel}]}
  MaterialExpressionReroute_60 {class=MaterialExpressionReroute, inputs [None {from=MaterialExpressionMultiply_60}]}
  MaterialExpressionReroute_62 {class=MaterialExpressionReroute, inputs [None {from=MaterialExpressionSaturate_23}]}
  MaterialExpressionReroute_63 {class=MaterialExpressionReroute, inputs [None {from=MaterialExpressionMax_8}]}
  MaterialExpressionReroute_65 {class=MaterialExpressionReroute, inputs [None {from=MaterialExpressionReroute_67}]}
  MaterialExpressionReroute_66 {class=MaterialExpressionReroute, inputs [None {from=MaterialExpressionMaterialFunctionCall_24, output=Center Orb Distance}]}
  MaterialExpressionReroute_67 {class=MaterialExpressionReroute, inputs [None {from=MaterialExpressionMaterialFunctionCall_24, output=FOV - Scale Multiplier}]}
  MaterialExpressionReroute_68 {class=MaterialExpressionReroute, inputs [None {from=MaterialExpressionGetMaterialAttributes_5, output=EmissiveColor}]}
  MaterialExpressionReroute_69 {class=MaterialExpressionReroute, inputs [None {from=MaterialExpressionMultiply_87}]}
  MaterialExpressionReroute_7 {class=MaterialExpressionReroute, inputs [None {from=MaterialExpressionReroute_6}]}
  MaterialExpressionReroute_70 {class=MaterialExpressionReroute, inputs [None {from=MaterialExpressionReroute_98}]}
  MaterialExpressionReroute_71 {class=MaterialExpressionReroute, inputs [None {from=MaterialExpressionSetMaterialAttributes_5}]}
  MaterialExpressionReroute_72 {class=MaterialExpressionReroute, inputs [None {from=MaterialExpressionMultiply_73}]}
  MaterialExpressionReroute_73 {class=MaterialExpressionReroute, inputs [None {from=MaterialExpressionReroute_72}]}
  MaterialExpressionReroute_74 {class=MaterialExpressionReroute, inputs [None {from=MaterialExpressionMultiply_78}]}
  MaterialExpressionReroute_75 {class=MaterialExpressionReroute, inputs [None {from=MaterialExpressionMaterialFunctionCall_33, output=RadialGradientExponential}]}
  MaterialExpressionReroute_76 {class=MaterialExpressionReroute, inputs [None {from=MaterialExpressionReroute_75}]}
  MaterialExpressionReroute_77 {class=MaterialExpressionReroute, inputs [None {from=MaterialExpressionReroute_76}]}
  MaterialExpressionReroute_78 {class=MaterialExpressionReroute, inputs [None {from=MaterialExpressionReroute_57}]}
  MaterialExpressionReroute_79 {class=MaterialExpressionReroute, inputs [None {from=MaterialExpressionReroute_56}]}
  MaterialExpressionReroute_8 {class=MaterialExpressionReroute, inputs [None {from=MaterialExpressionMaterialFunctionCall_21, output=Fresnel}]}
  MaterialExpressionReroute_80 {class=MaterialExpressionReroute, inputs [None {from=MaterialExpressionLinearInterpolate_27}]}
  MaterialExpressionReroute_81 {class=MaterialExpressionReroute, inputs [None {from=MaterialExpressionLinearInterpolate_28}]}
  MaterialExpressionReroute_82 {class=MaterialExpressionReroute, inputs [None {from=MaterialExpressionMax_7}]}
  MaterialExpressionReroute_83 {class=MaterialExpressionReroute, inputs [None {from=MaterialExpressionReroute_84}]}
  MaterialExpressionReroute_84 {class=MaterialExpressionReroute, inputs [None {from=MaterialExpressionReroute_18}]}
  MaterialExpressionReroute_85 {class=MaterialExpressionReroute, inputs [None {from=MaterialExpressionAdd_23}]}
  MaterialExpressionReroute_86 {class=MaterialExpressionReroute, inputs [None {from=MaterialExpressionMaterialFunctionCall_36, output=Orb Center World Pos}]}
  MaterialExpressionReroute_87 {class=MaterialExpressionReroute, inputs [None {from=MaterialExpressionReroute_88}]}
  MaterialExpressionReroute_88 {class=MaterialExpressionReroute, inputs [None {from=MaterialExpressionGetMaterialAttributes_3, output=MaterialAttributes}]}
  MaterialExpressionReroute_89 {class=MaterialExpressionReroute, inputs [None {from=MaterialExpressionSetMaterialAttributes_2}]}
  MaterialExpressionReroute_9 {class=MaterialExpressionReroute, inputs [None {from=MaterialExpressionFunctionInput_1}]}
  MaterialExpressionReroute_90 {class=MaterialExpressionReroute, inputs [None {from=MaterialExpressionAdd_24}]}
  MaterialExpressionReroute_91 {class=MaterialExpressionReroute, inputs [None {from=MaterialExpressionReroute_53}]}
  MaterialExpressionReroute_98 {class=MaterialExpressionReroute, inputs [None {from=MaterialExpressionReroute_71}]}
  MaterialExpressionReroute_99 {class=MaterialExpressionReroute, inputs [None {from=MaterialExpressionLinearInterpolate_26}]}
  MaterialExpressionRound_0 {class=MaterialExpressionRound, inputs [None {from=MaterialExpressionNamedRerouteUsage_40, output=""}], properties {}}
  MaterialExpressionSaturate_0 {class=MaterialExpressionSaturate, inputs [None {from=MaterialExpressionMultiply_76}]}
  MaterialExpressionSaturate_1 {class=MaterialExpressionSaturate, inputs [None {from=MaterialExpressionFunctionInput_0}]}
  MaterialExpressionSaturate_13 {class=MaterialExpressionSaturate, inputs [None {from=MaterialExpressionSubtract_33}]}
  MaterialExpressionSaturate_15 {class=MaterialExpressionSaturate, inputs [None {from=MaterialExpressionMax_6}]}
  MaterialExpressionSaturate_17 {class=MaterialExpressionSaturate, inputs [None {from=MaterialExpressionAdd_21}]}
  MaterialExpressionSaturate_18 {class=MaterialExpressionSaturate, inputs [None {from=MaterialExpressionNamedRerouteUsage_34}]}
  MaterialExpressionSaturate_19 {class=MaterialExpressionSaturate, inputs [None {from=MaterialExpressionNamedRerouteUsage_31}]}
  MaterialExpressionSaturate_20 {class=MaterialExpressionSaturate, inputs [None {from=MaterialExpressionMultiply_59}]}
  MaterialExpressionSaturate_21 {class=MaterialExpressionSaturate, inputs [None {from=MaterialExpressionMaterialFunctionCall_27, output=VGradient}]}
  MaterialExpressionSaturate_22 {class=MaterialExpressionSaturate, inputs [None {from=MaterialExpressionMultiply_72}]}
  MaterialExpressionSaturate_23 {class=MaterialExpressionSaturate, inputs [None {from=MaterialExpressionAdd_22}]}
  MaterialExpressionSaturate_24 {class=MaterialExpressionSaturate, inputs [None {from=MaterialExpressionMax_9}]}
  MaterialExpressionSaturate_25 {class=MaterialExpressionSaturate, inputs [None {from=MaterialExpressionNamedRerouteUsage_35}]}
  MaterialExpressionSaturate_26 {class=MaterialExpressionSaturate, inputs [None {from=MaterialExpressionSubtract_29}]}
  MaterialExpressionSaturate_27 {class=MaterialExpressionSaturate, inputs [None {from=MaterialExpressionMultiply_84}]}
  MaterialExpressionSaturate_4 {class=MaterialExpressionSaturate, inputs [None {from=MaterialExpressionSubtract_27}]}
  MaterialExpressionSaturate_8 {class=MaterialExpressionSaturate, inputs [None {from=MaterialExpressionSubtract_32}]}
  MaterialExpressionScalarParameter_0 {class=MaterialExpressionScalarParameter, properties {defaultValue=1, group=Liquid, parameterName=Constellation Liquid Mask Blend}}
  MaterialExpressionScalarParameter_1 {class=MaterialExpressionScalarParameter, properties {defaultValue=1, group=Liquid, parameterName=Cap Max Value}}
  MaterialExpressionScalarParameter_10 {class=MaterialExpressionScalarParameter, properties {defaultValue=500, group=Liquid, parameterName=Foam Fade Distance, sortPriority=1}}
  MaterialExpressionScalarParameter_11 {class=MaterialExpressionScalarParameter, properties {defaultValue=50, parameterName=Fade Fade Hardness}}
  MaterialExpressionScalarParameter_12 {class=MaterialExpressionScalarParameter, properties {defaultValue=1, group=Liquid, parameterName=Gem Edge Glow Exp, sortPriority=70}}
  MaterialExpressionScalarParameter_13 {class=MaterialExpressionScalarParameter, properties {defaultValue=1, group=Liquid, parameterName=Gem Edge Glow Brightness, sortPriority=70}}
  MaterialExpressionScalarParameter_15 {class=MaterialExpressionScalarParameter, properties {defaultValue=5, group=Liquid, parameterName=Foam Fade Offset, sortPriority=1}}
  MaterialExpressionScalarParameter_18 {class=MaterialExpressionScalarParameter, properties {defaultValue=0.02, group=Liquid, parameterName=Wave Intensity, sortPriority=7}}
  MaterialExpressionScalarParameter_19 {class=MaterialExpressionScalarParameter, properties {defaultValue=1, group=Liquid, parameterName=Transition Height, sortPriority=5}}
  MaterialExpressionScalarParameter_2 {class=MaterialExpressionScalarParameter, properties {defaultValue=0.5, group=Liquid, parameterName=Cap Min Value}}
  MaterialExpressionScalarParameter_23 {class=MaterialExpressionScalarParameter, properties {defaultValue=0.02, group=Liquid, parameterName=Wave Intensity, sortPriority=7}}
  MaterialExpressionScalarParameter_25 {class=MaterialExpressionScalarParameter, properties {defaultValue=0.02, group=Liquid, parameterName=Wave Intensity, sortPriority=7}}
  MaterialExpressionScalarParameter_26 {class=MaterialExpressionScalarParameter, properties {defaultValue=1, group=Liquid, parameterName=Edge Foam Thickness, sortPriority=24}}
  MaterialExpressionScalarParameter_27 {class=MaterialExpressionScalarParameter, properties {defaultValue=0.1, group=Liquid, parameterName=Wave Speed, sortPriority=8}}
  MaterialExpressionScalarParameter_28 {class=MaterialExpressionScalarParameter, properties {defaultValue=4, group=Liquid, parameterName=Wave Size, sortPriority=6}}
  MaterialExpressionScalarParameter_29 {class=MaterialExpressionScalarParameter, properties {defaultValue=1, group=Liquid, parameterName=Liquid Opacity}}
  MaterialExpressionScalarParameter_3 {class=MaterialExpressionScalarParameter, properties {defaultValue=2, group=Liquid, parameterName=Edge Brightness Exp}}
  MaterialExpressionScalarParameter_30 {class=MaterialExpressionScalarParameter, properties {defaultValue=1, group=Liquid, parameterName=Depth Gradient Tile, sortPriority=20}}
  MaterialExpressionScalarParameter_32 {class=MaterialExpressionScalarParameter, properties {defaultValue=0.001, group=Liquid, parameterName=Liquid Depth Warp Amount}}
  MaterialExpressionScalarParameter_33 {class=MaterialExpressionScalarParameter, properties {defaultValue=4, group=Liquid, parameterName=Constellation Liquid Mask Mult}}
  MaterialExpressionScalarParameter_34 {class=MaterialExpressionScalarParameter, properties {defaultValue=0.7, group=Liquid, parameterName=Liquid Thickness}}
  MaterialExpressionScalarParameter_35 {class=MaterialExpressionScalarParameter, properties {defaultValue=0.01, group=Liquid, parameterName=Cap Faked Reflection, sortPriority=19}}
  MaterialExpressionScalarParameter_36 {class=MaterialExpressionScalarParameter, properties {defaultValue=0.6, group=Liquid, parameterName=Cap Radius Max, sortPriority=12}}
  MaterialExpressionScalarParameter_37 {class=MaterialExpressionScalarParameter, properties {defaultValue=0.3, group=Liquid, parameterName=Cap Radius Min, sortPriority=11}}
  MaterialExpressionScalarParameter_38 {class=MaterialExpressionScalarParameter, properties {defaultValue=1, group=Liquid, parameterName=Cap Texture Tiling A, sortPriority=13}}
  MaterialExpressionScalarParameter_39 {class=MaterialExpressionScalarParameter, properties {defaultValue=0.66, group=Liquid, parameterName=Cap Texture Tiling B, sortPriority=14}}
  MaterialExpressionScalarParameter_4 {class=MaterialExpressionScalarParameter, properties {defaultValue=3, group=Liquid, parameterName=Liquid Emissive Color Multi}}
  MaterialExpressionScalarParameter_40 {class=MaterialExpressionScalarParameter, properties {group=Liquid, parameterName=Cap Offset, sortPriority=10}}
  MaterialExpressionScalarParameter_41 {class=MaterialExpressionScalarParameter, properties {defaultValue=1, group=Liquid, parameterName=Cap Glow Density, sortPriority=16}}
  MaterialExpressionScalarParameter_42 {class=MaterialExpressionScalarParameter, properties {group=Liquid, parameterName=DepthLineHeightOffset, sortPriority=19}}
  MaterialExpressionScalarParameter_43 {class=MaterialExpressionScalarParameter, properties {group=Liquid, parameterName=Edge Foam Offset, sortPriority=25}}
  MaterialExpressionScalarParameter_44 {class=MaterialExpressionScalarParameter, properties {defaultValue=1, group=Liquid, parameterName=Edge Foam Opacity, sortPriority=26}}
  MaterialExpressionScalarParameter_45 {class=MaterialExpressionScalarParameter, properties {defaultValue=20, group=Liquid, parameterName=Cap Edge Density, sortPriority=13}}
  MaterialExpressionScalarParameter_46 {class=MaterialExpressionScalarParameter, properties {defaultValue=0.9, group=Liquid, parameterName=Radial Mask B Size}}
  MaterialExpressionScalarParameter_5 {class=MaterialExpressionScalarParameter, properties {defaultValue=1, group=Liquid, parameterName=Foam Transition Height, sortPriority=5}}
  MaterialExpressionScalarParameter_6 {class=MaterialExpressionScalarParameter, properties {defaultValue=1, group=Liquid, parameterName=Fill Emissive Boost}}
  MaterialExpressionScalarParameter_7 {
   class=MaterialExpressionScalarParameter
   properties {defaultValue=3, group=Liquid, parameterName=Edge Foam Thickness Distance Multi, sortPriority=24}
  }
  MaterialExpressionScalarParameter_8 {class=MaterialExpressionScalarParameter, properties {defaultValue=1, group=Liquid, parameterName=Edge Foam Emissive Boost, sortPriority=26}}
  MaterialExpressionScreenPosition_1 {class=MaterialExpressionScreenPosition, inputs [], properties {}}
  MaterialExpressionSetMaterialAttributes_0 {
   class=MaterialExpressionSetMaterialAttributes
   inputs [
    MaterialAttributes {from="", output=""}
    BaseColor {from=MaterialExpressionReroute_10, output=""}
    EmissiveColor {from=MaterialExpressionDivide_0, output=""}
   ]
   properties {attributeSetTypes [#2, #1]}
  }
  MaterialExpressionSetMaterialAttributes_1 {
   class=MaterialExpressionSetMaterialAttributes
   inputs [
    MaterialAttributes {from=MaterialExpressionGetMaterialAttributes_0, output=MaterialAttributes}
    EmissiveColor {from=MaterialExpressionLinearInterpolate_4, output=""}
   ]
   properties {attributeSetTypes [#1]}
  }
  MaterialExpressionSetMaterialAttributes_2 {
   class=MaterialExpressionSetMaterialAttributes
   inputs [
    MaterialAttributes {from=MaterialExpressionGetMaterialAttributes_3, output=MaterialAttributes}
    BaseColor {from=MaterialExpressionLinearInterpolate_34, output=""}
    EmissiveColor {from=MaterialExpressionLinearInterpolate_33, output=""}
   ]
   properties {attributeSetTypes [#2, #1]}
  }
  MaterialExpressionSetMaterialAttributes_4 {
   class=MaterialExpressionSetMaterialAttributes
   inputs [
    MaterialAttributes {from=MaterialExpressionGetMaterialAttributes_6, output=MaterialAttributes}
    BaseColor {from=MaterialExpressionLinearInterpolate_38, output=""}
    EmissiveColor {from=MaterialExpressionLinearInterpolate_37, output=""}
   ]
   properties {attributeSetTypes [#2, #1]}
  }
  MaterialExpressionSetMaterialAttributes_5 {
   class=MaterialExpressionSetMaterialAttributes
   inputs [
    MaterialAttributes {from=MaterialExpressionGetMaterialAttributes_7, output=MaterialAttributes}
    BaseColor {from=MaterialExpressionMaterialFunctionCall_40, output=Result}
    EmissiveColor {from=MaterialExpressionMaterialFunctionCall_39, output=Result}
   ]
   properties {attributeSetTypes [#2, #1]}
  }
  MaterialExpressionSine_0 {class=MaterialExpressionSine, inputs [None {from=MaterialExpressionNamedRerouteUsage_28}]}
  MaterialExpressionSine_3 {class=MaterialExpressionSine, inputs [None {from=MaterialExpressionNamedRerouteUsage_27}]}
  MaterialExpressionStaticSwitchParameter_0 {
   class=MaterialExpressionStaticSwitchParameter
   inputs [
    True {from=MaterialExpressionMultiply_8, output=""}
    False {from=MaterialExpressionSaturate_17, output=""}
   ]
   properties {group=Liquid, parameterName=Use Foam Color for Cap Color, sortPriority=15}
  }
  MaterialExpressionSubtract_1 {
   class=MaterialExpressionSubtract
   inputs [
    A {from=MaterialExpressionComponentMask_0}
    B {from=MaterialExpressionScalarParameter_15}
   ]
  }
  MaterialExpressionSubtract_10 {
   class=MaterialExpressionSubtract
   inputs [
    A {from=MaterialExpressionMultiply_40}
    B {from=MaterialExpressionComponentMask_3}
   ]
  }
  MaterialExpressionSubtract_11 {
   class=MaterialExpressionSubtract
   inputs [
    A {from=MaterialExpressionAdd_7}
    B {from=MaterialExpressionReroute_84}
   ]
  }
  MaterialExpressionSubtract_13 {
   class=MaterialExpressionSubtract
   inputs [
    A {from=MaterialExpressionCameraPositionWS_0}
    B {from=MaterialExpressionReroute_90}
   ]
  }
  MaterialExpressionSubtract_14 {
   class=MaterialExpressionSubtract
   inputs [
    A {from=MaterialExpressionSubtract_10}
    B {from=MaterialExpressionMultiply_44}
   ]
   properties {constB=3}
  }
  MaterialExpressionSubtract_17 {
   class=MaterialExpressionSubtract
   inputs [
    A {from=MaterialExpressionAdd_7}
    B {from=MaterialExpressionSubtract_24}
   ]
  }
  MaterialExpressionSubtract_2 {
   class=MaterialExpressionSubtract
   inputs [
    A {from=MaterialExpressionWorldPosition_1, output=XYZ}
    B {from=MaterialExpressionMaterialFunctionCall_0, output=Orb Center World Pos}
   ]
  }
  MaterialExpressionSubtract_23 {
   class=MaterialExpressionSubtract
   inputs [
    A {from=MaterialExpressionSaturate_13}
    B {from=MaterialExpressionSaturate_8}
   ]
  }
  MaterialExpressionSubtract_24 {
   class=MaterialExpressionSubtract
   inputs [
    A {from=MaterialExpressionReroute_18}
    B {from=MaterialExpressionMultiply_1}
   ]
   properties {constB=15}
  }
  MaterialExpressionSubtract_25 {
   class=MaterialExpressionSubtract
   inputs [
    A {from=MaterialExpressionReroute_46}
    B {from=MaterialExpressionReroute_3}
   ]
   properties {constB=15}
  }
  MaterialExpressionSubtract_26 {
   class=MaterialExpressionSubtract
   inputs [
    A {from=MaterialExpressionScalarParameter_25}
    B {from=MaterialExpressionDivide_4}
   ]
  }
  MaterialExpressionSubtract_27 {
   class=MaterialExpressionSubtract
   inputs [
    A {from=MaterialExpressionAdd_16}
    B {from=MaterialExpressionReroute_83}
   ]
  }
  MaterialExpressionSubtract_28 {
   class=MaterialExpressionSubtract
   inputs [
    A {from=MaterialExpressionSaturate_19}
    B {from=MaterialExpressionSaturate_20}
   ]
  }
  MaterialExpressionSubtract_29 {
   class=MaterialExpressionSubtract
   inputs [
    A {from=MaterialExpressionMultiply_83}
    B {from=MaterialExpressionReroute_62}
   ]
  }
  MaterialExpressionSubtract_30 {
   class=MaterialExpressionSubtract
   inputs [
    A {from=MaterialExpressionScreenPosition_1, output=ViewportUV}
    B {from=MaterialExpressionMultiply_68}
   ]
  }
  MaterialExpressionSubtract_31 {
   class=MaterialExpressionSubtract
   inputs [
    A {from=MaterialExpressionSubtract_30}
    B {from=""}
   ]
   properties {constB=0.01}
  }
  MaterialExpressionSubtract_32 {
   class=MaterialExpressionSubtract
   inputs [
    A {from=MaterialExpressionSubtract_25}
    B {from=MaterialExpressionScalarParameter_43}
   ]
   properties {constB=0}
  }
  MaterialExpressionSubtract_33 {
   class=MaterialExpressionSubtract
   inputs [
    A {from=MaterialExpressionDivide_1}
    B {from=MaterialExpressionScalarParameter_43}
   ]
   properties {constB=0}
  }
  MaterialExpressionSubtract_9 {
   class=MaterialExpressionSubtract
   inputs [
    A {from=MaterialExpressionWorldPosition_0, output=XYZ}
    B {from=MaterialExpressionMaterialFunctionCall_36, output=Orb Center World Pos}
   ]
  }
  MaterialExpressionTextureCoordinate_0 {class=MaterialExpressionTextureCoordinate}
  MaterialExpressionTextureCoordinate_3 {class=MaterialExpressionTextureCoordinate}
  MaterialExpressionTextureSample_4 {
   class=MaterialExpressionTextureSample
   inputs [
    UVs {from=MaterialExpressionAdd_15}
    Tex {from=""}
    Apply View MipBias {from=""}
   ]
   properties {samplerType=SAMPLERTYPE_LinearGrayscale, texture=&T_SimplePerlin_Greyscale}
  }
  MaterialExpressionTextureSample_5 {
   class=MaterialExpressionTextureSample
   inputs [
    UVs {from=MaterialExpressionAdd_17}
    Tex {from=""}
    Apply View MipBias {from=""}
   ]
   properties {samplerType=SAMPLERTYPE_LinearGrayscale, texture=/Game/Global/Textures/Default/Flowmaps/T_StormSwirlMask}
  }
  MaterialExpressionTextureSample_6 {
   class=MaterialExpressionTextureSample
   inputs [
    UVs {from=MaterialExpressionAdd_18}
    Tex {from=""}
    Apply View MipBias {from=""}
   ]
   properties {samplerType=SAMPLERTYPE_LinearGrayscale, texture=&T_StormSwirlMask}
  }
  MaterialExpressionTextureSample_7 {
   class=MaterialExpressionTextureSample
   inputs [
    UVs {from=MaterialExpressionAdd_20}
    Tex {from=""}
    Apply View MipBias {from=""}
   ]
   properties {samplerType=SAMPLERTYPE_LinearGrayscale, texture=&T_StormSwirlMask}
  }
  MaterialExpressionTime_2 {class=MaterialExpressionTime}
  MaterialExpressionTime_4 {class=MaterialExpressionTime}
  MaterialExpressionTime_5 {class=MaterialExpressionTime}
  MaterialExpressionTransform_0 {
   class=MaterialExpressionTransform
   inputs [None {from=MaterialExpressionVectorParameter_11, output=RGB}]
   properties {transformSourceType=TRANSFORMSOURCE_Local}
  }
  MaterialExpressionVectorParameter_10 {
   class=MaterialExpressionVectorParameter
   properties {group=Liquid, parameterName=LiquidAreaBounds, sortPriority=50, channelNames {a=Multiplier, b=Z, g=Y, r=X}, defaultValue=<rgba 1 1 1 160>}
  }
  MaterialExpressionVectorParameter_11 {
   class=MaterialExpressionVectorParameter
   properties {group=Liquid, parameterName=Liquid Center Offset, sortPriority=51, defaultValue=<rgba 0 0 0 1>}
  }
  MaterialExpressionVectorParameter_12 {
   class=MaterialExpressionVectorParameter
   properties {group=Liquid, parameterName=Cap Texture Speed, sortPriority=15, defaultValue=<rgba 0.5 0.45 0.55 -0.66>}
  }
  MaterialExpressionVectorParameter_13 {class=MaterialExpressionVectorParameter, properties {group=Liquid, parameterName=Background FIll Tint, defaultValue=<rgba 1 1 1 1>}}
  MaterialExpressionVectorParameter_7 {
   class=MaterialExpressionVectorParameter
   properties {group=Liquid, parameterName=Liquid Color, sortPriority=2, defaultValue=<rgba 0.0022 0.0469 0.0109 1>}
  }
  MaterialExpressionVectorParameter_8 {
   class=MaterialExpressionVectorParameter
   properties {group=Liquid, parameterName=Foam Color, sortPriority=31, defaultValue=<rgba 0 0.1302 0.0381 1>}
  }
  MaterialExpressionVectorParameter_9 {
   class=MaterialExpressionVectorParameter
   properties {group=Liquid, parameterName=Cap Glow Color, sortPriority=18, defaultValue=<rgba 1 0.7023 0 1>}
  }
  MaterialExpressionWorldPosition_0 {class=MaterialExpressionWorldPosition, inputs [], properties {}}
  MaterialExpressionWorldPosition_1 {class=MaterialExpressionWorldPosition, inputs [], properties {}}
 }}
