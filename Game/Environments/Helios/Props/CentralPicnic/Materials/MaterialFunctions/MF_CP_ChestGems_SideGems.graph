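# Material function that shades the CentralPicnic chest's side gems as glassy orbs with fake interior shards, blended over input attributes by Mask when UseSideGems is on.
# Each side, chosen by the rounded Left Right Mask, samples a packed mask on three object-aligned virtual planes at different depths.
# View-dependent radial masks, gradients and Fresnel shape the two-color interior.
# Glass color, thickness and roughness come from MF_CP_ChestGems_Utilities; an unmasked output is also exposed.
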MaterialFunction /Game/Environments/Helios/Props/CentralPicnic/Materials/MaterialFunctions/MF_CP_ChestGems_SideGems
graph {nodes [
  FeatureLevelSwitch(Default=QualitySwitch() {desc=Was combine virtual planes on low end to one plane}) {desc="Remove upper gradients on low ~26 inst"}
  FeatureLevelSwitch(Default=QualitySwitch()) {desc=combine virtual planes on low end to one plane}
  n1 = FunctionInput() {inputName=Material, inputType=FunctionInput_MaterialAttributes, sortPriority=1}
  n2 = FunctionInput(Preview=Constant() {r=1}) {inputName=Mask, inputType=FunctionInput_Scalar, sortPriority=2}
  n3 = MaterialFunctionCall!(WarpMask) {materialFunction=&/MF_CP_ChestGems_Utilities}
  n4 = VectorParameter() {group=_Colors, parameterName=Side Orb Background Color, sortPriority=27, defaultValue=<rgba 0.1354 0.0097 0.0007 1>}
  n5 = VectorParameter() {group=_Colors, parameterName=Side Orb Gradient Color, sortPriority=28, defaultValue=<rgba 0.8958 0.2184 0.0663 1>}
  n6 = FunctionInput() {description="White (1) is Side A, Black (0) is Side B", inputName=Left Right Mask, inputType=FunctionInput_Scalar, sortPriority=3}
  n7 = Round(None=@n6)
  n8 = TextureObjectParameter() {
   group=Side Orb | Shard Layers
   parameterName=Layers Packed Mask
   sortPriority=10
   texture=/Game/Environments/Helios/Props/CentralPicnic/Textures/T_CentralPicnic_ChestSideOrbMask
  }
  n9 = NamedRerouteUsage()
  n10 = VectorParameter() {
   group=Side Orb | Shard Layers
   parameterName=Shard Plane Depths
   sortPriority=21
   channelNames {a=Offset, b=Depth 3, g=Depth 2, r=Depth 1}
   defaultValue=<rgba 80 40 0 1>
  }
  n11 = MaterialFunctionCall() {materialFunction=&ObjectScale}
  n12 = Multiply(A=Multiply(A=ScalarParameter() {defaultValue=500, group=Side Orb | Shard Layers, parameterName=Scale, sortPriority=25}, B=@n11.Scale X)) {constB=-1}
  n13 = MaterialFunctionCall!(
   Projection Pivot Position=@n9
   Plane Depth Offset=Multiply(A=Add(A=@n10.Depth 1, B=@n10.Depth 1), B=@n11.Scale X)
   "Coordinate Scale (WS)"=@n12
   World Space Projection Normal
   World Space Projection Tangent
   World Space Projection Binormal
  ) {materialFunction=/Engine/Functions/Engine_MaterialFunctions02/UVs/ObjectAlignedVirtualPlaneCoordinates}
  n14 = NamedRerouteUsage()
  n15 = Max(A=ComponentMask(None=TextureSample(UVs=@n13.Center Aligned UV Plane, Tex=@n8) {samplerType=SAMPLERTYPE_LinearColor}.RGB) {b=true}, B=@n14)
  n16 = VectorParameter() {
   desc=Alpha is Global Multiplier
   group=Side Orb | Shard Layers
   parameterName=Layer Intensities
   sortPriority=19
   channelNames {a=Master, b=3, g=2, r=1}
   defaultValue=<rgba 0.5 1 0.25 1>
  }
  n17 = Multiply(A=@n16.2, B=@n16.2)
  n18 = Add(A=@n10.Depth 3, B=@n10.Depth 3)
  n19 = TextureSample(
   UVs=MaterialFunctionCall(Projection Pivot Position=@n9, Plane Depth Offset=Multiply(A=@n11.Scale X, B=@n18), "Coordinate Scale (WS)"=@n12) {materialFunction=&ObjectAlignedVirtualPlaneCoordinates}.Center Aligned UV Plane
   Tex=@n8
  ) {samplerType=SAMPLERTYPE_LinearColor}
  n20 = Multiply(A=Max(A=ComponentMask(None=@n19.RGB) {g=true}, B=@n14), B=Multiply(A=@n16.3, B=@n16.3)) {constB=0.1}
  n21 = Add(A=@n10.Depth 2, B=@n10.Depth 2)
  n22 = TextureSample(
   UVs=MaterialFunctionCall(Projection Pivot Position=@n9, Plane Depth Offset=Multiply(A=@n11.Scale X, B=@n21), "Coordinate Scale (WS)"=@n12) {materialFunction=&ObjectAlignedVirtualPlaneCoordinates}.Center Aligned UV Plane
   Tex=@n8
  ) {samplerType=SAMPLERTYPE_LinearColor}
  n23 = Multiply(A=Max(A=@n14, B=ComponentMask(None=@n22.RGB) {r=true}), B=Multiply(A=@n16.1, B=@n16.1)) {constB=0.25}
  n24 = Add(A=Add(A=Multiply(A=@n15, B=@n17) {constB=0.1}, B=Add(A=@n20, B=@n23)), B=NamedRerouteUsage())
  n25 = TextureObjectParameter() {
   group=Side Orb | Shard Layers
   parameterName=Layers Packed Mask
   sortPriority=10
   texture=&T_CentralPicnic_ChestSideOrbMask
  }
  n26 = NamedRerouteUsage()
  n27 = Normalize(VectorInput=Transform(None=Constant3Vector() {constant=<rgba -1 0 0 0>}) {transformSourceType=TRANSFORMSOURCE_Instance})
  n28 = VectorParameter() {
   group=Side Orb | Shard Layers
   parameterName=Shard Plane Depths
   sortPriority=21
   channelNames {a=Offset, b=Depth 3, g=Depth 2, r=Depth 1}
   defaultValue=<rgba 80 40 0 1>
  }
  n29 = MaterialFunctionCall() {materialFunction=&ObjectScale}
  n30 = Multiply(A=Multiply(A=ScalarParameter() {defaultValue=500, group=Side Orb | Shard Layers, parameterName=Scale, sortPriority=25}, B=@n29.Scale X)) {constB=-1}
  n31 = MaterialFunctionCall(
   Projection Pivot Position=@n26
   Plane Depth Offset=Multiply(A=Add(A=@n28.Depth 1, B=@n28.Depth 1), B=@n29.Scale X)
   "Coordinate Scale (WS)"=@n30
   World Space Projection Normal=@n27
  ) {materialFunction=&ObjectAlignedVirtualPlaneCoordinates}
  n32 = NamedRerouteUsage()
  n33 = Max(A=ComponentMask(None=TextureSample(UVs=@n31.Center Aligned UV Plane, Tex=@n25) {samplerType=SAMPLERTYPE_LinearColor}.RGB) {b=true}, B=@n32)
  n34 = VectorParameter() {
   desc=Alpha is Global Multiplier
   group=Side Orb | Shard Layers
   parameterName=Layer Intensities
   sortPriority=19
   defaultValue=<rgba 0.6 1 0.3 1>
  }
  n35 = Multiply(A=@n34.G, B=@n34.G)
  n36 = Add(A=@n28.Depth 3, B=@n28.Depth 3)
  n37 = TextureSample(
   UVs=MaterialFunctionCall(
    Projection Pivot Position=@n26
    Plane Depth Offset=Multiply(A=@n36, B=@n29.Scale X)
    "Coordinate Scale (WS)"=@n30
    World Space Projection Normal=@n27
   ) {materialFunction=&ObjectAlignedVirtualPlaneCoordinates}.Center Aligned UV Plane
   Tex=@n25
  ) {samplerType=SAMPLERTYPE_LinearColor}
  n38 = Multiply(A=Max(A=ComponentMask(None=@n37.RGB) {g=true}, B=@n32), B=Multiply(A=@n34.B, B=@n34.B)) {constB=0.1}
  n39 = Add(A=@n28.Depth 2, B=@n28.Depth 2)
  n40 = TextureSample(
   UVs=MaterialFunctionCall(
    Projection Pivot Position=@n26
    Plane Depth Offset=Multiply(A=@n39, B=@n29.Scale X)
    "Coordinate Scale (WS)"=@n30
    World Space Projection Normal=@n27
   ) {materialFunction=&ObjectAlignedVirtualPlaneCoordinates}.Center Aligned UV Plane
   Tex=@n25
  ) {samplerType=SAMPLERTYPE_LinearColor}
  n41 = Multiply(A=Max(A=@n32, B=ComponentMask(None=@n40.RGB) {r=true}), B=Multiply(A=@n34.R, B=@n34.R)) {constB=0.25}
  n42 = Add(A=Add(A=Multiply(A=@n33, B=@n35) {constB=0.1}, B=Add(A=@n38, B=@n41)), B=NamedRerouteUsage())
  n43 = Fresnel() {baseReflectFraction=0, exponent=1}
  n44 = Multiply(
   A=Max(A=LinearInterpolate(B=@n24, Alpha=@n7) {constB=0}, B=LinearInterpolate(A=@n42, Alpha=@n7) {constB=0})
   B=Power(
    Base=OneMinus(None=@n43)
    Exp=ScalarParameter() {defaultValue=1, group=Side Orb | Shard Layers, parameterName=Orb Fresnel Shading, sortPriority=29}
   ) {constExponent=1}
  )
  n45 = Multiply(
   A=LinearInterpolate(A=@n4.RGB, B=@n5.RGB, Alpha=NamedRerouteDeclaration(None=@n44) {name=SideOrb ShardValue, nodeColor=<rgba 0 0.0625 1 1>})
   B=MaterialFunctionCall(Exponent=ScalarParameter() {defaultValue=1.75, group=Side Orb | General, parameterName=Side Orb Fresnel Exponent}) {materialFunction=&MF_Fresnel_FX}.Fresnel
  )
  n46 = Multiply(A=@n45) {constB=0.25}
  n47 = MakeMaterialAttributes!(
   BaseColor=LinearInterpolate(A=@n45, B=@n3.Glass Color, Alpha=@n3.Glass Thickness Mask)
   Metallic
   Specular
   Roughness=@n3.Gem Roughness
   Anisotropy
   EmissiveColor=Multiply(
    A=LinearInterpolate(A=@n45, B=@n46, Alpha=@n3.Glass Thickness Mask) {constB=0}
    B=ScalarParameter() {defaultValue=0.25, group=Side Orb | General, parameterName=Side Orb Emissive Power}
   )
   Opacity
   OpacityMask
   Normal
   Tangent
   WorldPositionOffset
   SubsurfaceColor
   ClearCoat
   ClearCoatRoughness
   AmbientOcclusion
   Refraction
   CustomizedUVs_0
   CustomizedUVs_1
   CustomizedUVs_2
   CustomizedUVs_3
   CustomizedUVs_4
   CustomizedUVs_5
   CustomizedUVs_6
   CustomizedUVs_7
   PixelDepthOffset
   ShadingModel
   Displacement
  )
  FunctionOutput(None=StaticSwitchParameter(True=BlendMaterialAttributes(A=@n1, B=@n47, Alpha=@n2), False=@n1) {defaultValue=true, group=_General, parameterName=UseSideGems, sortPriority=2}) {outputName=Material}
  FunctionOutput(None=@n47) {outputName=Material Non Masked, sortPriority=1}
  NamedRerouteDeclaration(None=@n3.FOV - Scale Multiplier) {name=FOV Scalar, nodeColor=<rgba 1 0.3281 0 1>}
  n48 = Add(
   A=MaterialFunctionCall(UV Channel=NamedRerouteUsage()) {materialFunction=&LinearGradient}.VGradient
   B=ScalarParameter() {group=Side Orb | Gradient, parameterName=SideOrb Gradient Height Offset}
  ) {constB=0.2}
  n49 = ScalarParameter() {defaultValue=1, group=Side Orb | Gradient, parameterName=SideOrb Gradient Contrast}
  n50 = OneMinus(None=Saturate(None=Power(Base=@n48, Exp=@n49) {constExponent=5}))
  n51 = NamedRerouteUsage()
  n52 = LinearInterpolate(A=@n50, Alpha=Max(A=Multiply(A=@n50, B=@n51), B=@n51))
  NamedRerouteDeclaration(None=Multiply(A=Saturate(None=@n52), B=ScalarParameter() {defaultValue=0.2, group=Side Orb | Gradient, parameterName=SideOrb Gradient Power}) {constB=5}) {name=SideOrb HorizontalGradientSide 2, nodeColor=<rgba 1 0.2578 0 1>}
  n53 = ScalarParameter() {defaultValue=1, desc=Perpendicular to the actor, group=Side Orb | Shard Layers, parameterName=Mask Radius B, sortPriority=37}
  n54 = ScalarParameter() {defaultValue=1, desc=Directly facing the actor, group=Side Orb | Shard Layers, parameterName=Mask Radius A, sortPriority=36}
  n55 = Normalize(VectorInput=VectorParameter() {group=_General, parameterName=Orb Front Direction, defaultValue=<rgba 0 1 0 1>}.RGB)
  n56 = Subtract(A=CameraPositionWS(), B=ObjectPositionWS())
  n57 = Abs(None=DotProduct(A=Transform(None=@n55) {transformSourceType=TRANSFORMSOURCE_Local}, B=Normalize(VectorInput=@n56)))
  n58 = MaterialFunctionCall() {materialFunction=&ObjectScale}
  n59 = VectorParameter() {
   group=Side Orb | Shard Layers
   parameterName=Shard Plane Depths
   sortPriority=21
   channelNames {a=Offset, b=Depth 3, g=Depth 2, r=Depth 1}
   defaultValue=<rgba 80 40 0 1>
  }
  n60 = Max(A=Max(A=@n59.Depth 1, B=@n59.Depth 1), B=@n59.Depth 3)
  n61 = Min(A=@n59.Depth 2, B=Min(A=@n59.Depth 1, B=@n59.Depth 1))
  n62 = Multiply(
   A=Add(A=LinearInterpolate(A=@n60, B=@n61), B=@n59.Offset)
   B=MaterialFunctionCall() {materialFunction=&ObjectScale}.Scale X
  )
  n63 = MaterialFunctionCall(
   Projection Pivot Position=NamedRerouteUsage()
   Plane Depth Offset=@n62
   "Coordinate Scale (WS)"=ScalarParameter() {defaultValue=500, group=Side Orb | Shard Layers, parameterName=Scale, sortPriority=25}
  ) {materialFunction=&ObjectAlignedVirtualPlaneCoordinates}
  n64 = MaterialFunctionCall(
   X=ScalarParameter() {group=Side Orb | Shard Layers, parameterName=Radial Plane Mask Offset X, sortPriority=40}
   Y=ScalarParameter() {group=Side Orb | Shard Layers, parameterName=Radial Plane Mask Offset Y, sortPriority=41}
  ) {materialFunction=&MakeFloat2}
  n65 = MaterialFunctionCall(
   UVs=Add(
    A=NamedRerouteDeclaration(None=@n63.Center Aligned UV Plane) {name=SideOrb AvergePlaneUVs Side 1, nodeColor=<rgba 0.9688 0 1 1>}
    B=Multiply(A=@n58.Scale X, B=@n64.Result)
   )
   Radius=Multiply(A=LinearInterpolate(A=@n53, B=@n54, Alpha=@n57), B=@n58.Scale X)
   Density=ScalarParameter() {defaultValue=3, group=Side Orb | Shard Layers, parameterName=Radial Mask Density, sortPriority=38}
  ) {materialFunction=&RadialGradientExponential}
  NamedRerouteDeclaration(None=Saturate(None=OneMinus(None=@n65.RadialGradientExponential))) {name=SideOrb VirtualPlaneMaskSide 1, nodeColor=<rgba 0 1 0.6719 1>}
  n66 = Add(
   A=MaterialFunctionCall(UV Channel=NamedRerouteUsage()) {materialFunction=&LinearGradient}.VGradient
   B=ScalarParameter() {group=Side Orb | Gradient, parameterName=SideOrb Gradient Height Offset}
  ) {constB=0.2}
  n67 = ScalarParameter() {defaultValue=1, group=Side Orb | Gradient, parameterName=SideOrb Gradient Contrast}
  n68 = OneMinus(None=Saturate(None=Power(Base=@n66, Exp=@n67) {constExponent=5}))
  n69 = NamedRerouteUsage()
  n70 = LinearInterpolate(A=@n68, Alpha=Max(A=Multiply(A=@n68, B=@n69), B=@n69))
  NamedRerouteDeclaration(None=Multiply(A=Saturate(None=@n70), B=ScalarParameter() {defaultValue=0.2, group=Side Orb | Gradient, parameterName=SideOrb Gradient Power}) {constB=5}) {name=SideOrb HorizontalGradientSide 1, nodeColor=<rgba 1 0.2578 0 1>}
  n71 = ScalarParameter() {defaultValue=1, desc=Perpendicular to the actor, group=Side Orb | Shard Layers, parameterName=Mask Radius B, sortPriority=37}
  n72 = ScalarParameter() {defaultValue=1, desc=Directly facing the actor, group=Side Orb | Shard Layers, parameterName=Mask Radius A, sortPriority=36}
  n73 = Normalize(VectorInput=VectorParameter() {group=_General, parameterName=Orb Front Direction, defaultValue=<rgba 0 1 0 1>}.RGB)
  n74 = Subtract(A=CameraPositionWS(), B=ObjectPositionWS())
  n75 = Abs(None=DotProduct(A=Transform(None=@n73) {transformSourceType=TRANSFORMSOURCE_Local}, B=Normalize(VectorInput=@n74)))
  n76 = MaterialFunctionCall(
   Projection Pivot Position=NamedRerouteUsage()
   Plane Depth Offset=@n62
   "Coordinate Scale (WS)"=ScalarParameter() {defaultValue=500, group=Side Orb | Shard Layers, parameterName=Scale, sortPriority=25}
   World Space Projection Normal=@n27
  ) {materialFunction=&ObjectAlignedVirtualPlaneCoordinates}
  n77 = MaterialFunctionCall(
   X=ScalarParameter() {group=Side Orb | Shard Layers, parameterName=Radial Plane Mask Offset X, sortPriority=40}
   Y=ScalarParameter() {group=Side Orb | Shard Layers, parameterName=Radial Plane Mask Offset Y, sortPriority=41}
  ) {materialFunction=&MakeFloat2}
  n78 = NamedRerouteUsage()
  n79 = MaterialFunctionCall(
   UVs=Add(
    A=NamedRerouteDeclaration(None=@n76.Center Aligned UV Plane) {name=SideOrb AvergePlaneUVs Side 2, nodeColor=<rgba 0 1 0.625 1>}
    B=Multiply(A=@n77.Result, B=@n78)
   )
   Radius=Multiply(A=LinearInterpolate(A=@n71, B=@n72, Alpha=@n75), B=NamedRerouteUsage())
   Density=ScalarParameter() {defaultValue=3, group=Side Orb | Shard Layers, parameterName=Radial Mask Density, sortPriority=38}
  ) {materialFunction=&RadialGradientExponential}
  NamedRerouteDeclaration(None=Saturate(None=OneMinus(None=@n79.RadialGradientExponential))) {name=SideOrb VirtualPlaneMaskSide 2, nodeColor=<rgba 0 1 0.6719 1>}
  n80 = VectorParameter() {group=Side Orb | Orb Mask, parameterName=Side Orb Plane Height, defaultValue=<rgba 0 0 0 1>}
  NamedRerouteDeclaration(None=Add(A=ObjectPositionWS(), B=Transform(None=@n80.RGB) {transformSourceType=TRANSFORMSOURCE_Local}) {constB=0}) {name=SideOrb ShardPlanePosition, nodeColor=<rgba 0 0.8594 1 1>}
  n81 = MaterialFunctionCall(Float3=@n22.RGB) {materialFunction=&BreakOutFloat3Components}
  n82 = NamedRerouteUsage()
  n83 = VectorParameter() {
   desc=Alpha is Global Multiplier
   group=Side Orb | Shard Layers
   parameterName=Layer Intensities
   sortPriority=19
   defaultValue=<rgba 0.6 1 0.3 1>
  }
  n84 = Multiply(A=Max(A=@n81.R, B=@n82), B=Multiply(A=@n83.R, B=@n83.R))
  n85 = Multiply(A=Max(A=@n81.G, B=@n82), B=Multiply(A=@n83.B, B=@n83.B))
  n86 = Max(A=@n81.B, B=@n82)
  n87 = Multiply(A=@n83.G, B=@n83.G)
  n88 = Add(A=Add(A=Add(A=@n84, B=@n85), B=Multiply(A=@n86, B=@n87)), B=NamedRerouteUsage())
  Reroute(None=@n88)
  Reroute(None=@n88)
  n89 = MaterialFunctionCall(Float3=@n40.RGB) {materialFunction=&BreakOutFloat3Components}
  n90 = NamedRerouteUsage()
  n91 = VectorParameter() {
   desc=Alpha is Global Multiplier
   group=Side Orb | Shard Layers
   parameterName=Layer Intensities
   sortPriority=19
   defaultValue=<rgba 0.6 1 0.3 1>
  }
  n92 = Multiply(A=Max(A=@n89.R, B=@n90), B=Multiply(A=@n91.R, B=@n91.R))
  n93 = Multiply(A=Max(A=@n89.G, B=@n90), B=Multiply(A=@n91.B, B=@n91.B))
  n94 = Max(A=@n89.B, B=@n90)
  n95 = Multiply(A=@n91.G, B=@n91.G)
  n96 = Add(A=Add(A=Add(A=@n92, B=@n93), B=Multiply(A=@n94, B=@n95)), B=NamedRerouteUsage())
  Reroute(None=@n96)
  Reroute(None=@n96)
 ]}
