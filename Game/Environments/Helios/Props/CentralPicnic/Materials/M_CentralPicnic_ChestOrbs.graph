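# Material for the Central Picnic chest orbs, building the look of its center, side and accent gems with Nanite support.
# A mask function separates the gem regions, which chained material-attribute functions layer from center through side to accent.
# A static switch selects a plain or liquid-filled center gem, with fill from a parameter or collection and edge glow from the collection.
# The result passes through destructible damage, offering optional cracks and default-on wobble.
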Material /Game/Environments/Helios/Props/CentralPicnic/Materials/M_CentralPicnic_ChestOrbs
graph {
 properties {bUseMaterialAttributes=true, bUsedWithNanite=true, pixelDepthOffsetMode=PDOM_Legacy, usedShadingModels=Default Lit}
 nodes [
  n1 = MaterialFunctionCall() {materialFunction=/Game/Environments/Helios/Props/CentralPicnic/Materials/MaterialFunctions/MF_CP_ChestGems_Masks}
  n2 = CollectionParameter() {collection=/Game/Environments/Helios/Props/CentralPicnic/Materials/MaterialFunctions/MPC_ChestGems, parameterName=FillAlpha}
  n3 = ScalarParameter() {group=Liquid, parameterName=Fill Alpha, sortPriority=6}
  n4 = StaticSwitchParameter(
   True=MaterialFunctionCall!(
    In
    Center Gem Mask=@n1.Center Gem
    Fill Alpha=StaticSwitchParameter(True=@n2, False=@n3) {group=Liquid, parameterName=UseLiquidFillMPC, sortPriority=3}
    Gem Edge Glow Multi=CollectionParameter() {collection=&MPC_ChestGems, parameterName=GemEdgeGlowMulti}
    Gem Edge Glow Color=CollectionParameter() {collection=&MPC_ChestGems, parameterName=GemEdgeGlowColor}
   ) {materialFunction=/Game/Environments/Helios/Props/CentralPicnic/Materials/MaterialFunctions/MF_CP_ChestGems_CenterGemWithLiquid}.Result
   False=MaterialFunctionCall!(Material, Mask=@n1.Center Gem, WarpMask, Constellation Wisp Mask Opacity) {materialFunction=/Game/Environments/Helios/Props/CentralPicnic/Materials/MaterialFunctions/MF_CP_ChestGems_CenterGem}.Material
  ) {group=Liquid, parameterName=Use Center Gem Liquid, sortPriority=5}
  n5 = MaterialFunctionCall() {materialFunction=&MF_CP_ChestGems_Masks}
  MaterialFunctionCall(
   Material_In=MaterialFunctionCall!(
    Material=MaterialFunctionCall!(Material=@n4, Mask=@n5.Side Gems Combined, Left Right Mask=@n5.Side Gems Combined) {materialFunction=/Game/Environments/Helios/Props/CentralPicnic/Materials/MaterialFunctions/MF_CP_ChestGems_SideGems}.Material
    Mask=@n5.Accent Gems
   ) {materialFunction=/Game/Environments/Helios/Props/CentralPicnic/Materials/MaterialFunctions/MF_CP_ChestGems_AccentGems}.Material
   UseCrackedDamageTextures=StaticBoolParameter() {group=z_GameplayParameters_DestructibleDamage, parameterName=Use Crack Damage Textures, sortPriority=2}
   UseDamageWobbleWPO=StaticBoolParameter() {defaultValue=true, group=z_GameplayParameters_DestructibleDamage, parameterName=Use Damage Wobble, sortPriority=1}
  ) {materialFunction=&MF_FN_DestructibleDamage}
 ]
}
