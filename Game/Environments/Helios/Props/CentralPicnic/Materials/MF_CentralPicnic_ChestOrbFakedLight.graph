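# Material function that fakes the chest orb's glow on CentralPicnic props, blended into emissive behind a static switch.
# Pivot-relative center drives normal-based diffuse, plus camera fresnel, tinted by one color.
# Sphere masks over flowmap-warped world positions form inner and ring shapes with sine flicker, scaled by time of day and also output as Mask.
# Ambient occlusion masks the blend, an optional chain switch modulates by Metallic, and MaterialProxyReplace supplies simplified HLOD values.
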MaterialFunction /Game/Environments/Helios/Props/CentralPicnic/Materials/MF_CentralPicnic_ChestOrbFakedLight
graph {nodes [
  n1 = FunctionInput() {inputName=Material Attributes, inputType=FunctionInput_MaterialAttributes}
  n2 = GetMaterialAttributes(None=@n1) {attributeGetTypes [
    #1
    #2
    #3
    #4
   ]}
  n3 = CameraVectorWS()
  n4 = Constant3Vector() {constant=<rgba 0 0 1 0>}
  n5 = Normalize(VectorInput=FunctionInput(Preview=Transform(None=@n4)) {bUsePreviewValueAsDefault=true, inputName=World Normal Map, sortPriority=2})
  n6 = Saturate(None=DotProduct(A=Normalize(VectorInput=@n3), B=@n5))
  n7 = ScalarParameter() {defaultValue=1, group=Chest Light, parameterName=Fresnel Contrast, sortPriority=36}
  n8 = Saturate(None=Multiply(A=Power(Base=@n6, Exp=@n7), B=ScalarParameter() {group=Chest Light, parameterName=Fresnel Brightness, sortPriority=37}))
  n9 = VectorParameter() {group=Chest Light, parameterName=CenterPoint, sortPriority=10, defaultValue=<rgba 0 0 0 1>}
  n10 = Transform(None=VectorParameter() {group=Chest Light, parameterName=HLODCenterPoint, sortPriority=10, defaultValue=<rgba 0 0 0 1>}.RGB) {transformSourceType=TRANSFORMSOURCE_Local}
  n11 = MaterialFunctionCall() {materialFunction=&ComponentPivotLocation}
  n12 = MaterialProxyReplace(
   Realtime=Add(
    A=Transform(None=@n9.RGB) {transformSourceType=TRANSFORMSOURCE_Local}
    B=MaterialFunctionCall() {materialFunction=&ComponentPivotLocation}.Result
   )
   MaterialProxy=TransformPosition(None=Add(A=@n10, B=@n11.Result)) {transformSourceType=TRANSFORMPOSSOURCE_World}
  )
  n13 = Normalize(VectorInput=Subtract(A=@n12, B=WorldPosition().XYZ))
  n14 = WorldPosition()
  n15 = VertexNormalWS()
  n16 = DotProduct(
   A=Normalize(VectorInput=Subtract(A=@n12, B=@n14.XYZ))
   B=Normalize(VectorInput=FunctionInput(Preview=@n15) {bUsePreviewValueAsDefault=true, inputName=Vertex Normals, sortPriority=3})
  )
  n17 = Constant() {r=1}
  n18 = Constant()
  n19 = FunctionInput(Preview=StaticBool() {value=true}) {bUsePreviewValueAsDefault=true, inputName=Use Vertex Normal Masking, inputType=FunctionInput_StaticBool, sortPriority=4}
  n20 = LinearInterpolate(
   B=Saturate(None=DotProduct(A=@n13, B=@n5))
   Alpha=LinearInterpolate(B=Saturate(None=@n16), Alpha=StaticSwitch(True=@n17, False=@n18, Value=@n19)) {constA=1}
  )
  n21 = VectorParameter() {group=Chest Light, parameterName=Color A, sortPriority=20, defaultValue=<rgba 1 0.263 0.032 1>}
  n22 = MaterialProxyReplace(Realtime=Multiply(A=Add(A=@n8, B=@n20), B=@n21.RGB) {desc=Color}, MaterialProxy=@n21.RGB)
  n23 = ScalarParameter() {defaultValue=10, group=Chest Light, parameterName=Master Intensity, sortPriority=22}
  n24 = ScalarParameter() {defaultValue=4, group=Chest Light, parameterName=HLOD Master Intensity, sortPriority=22}
  n25 = ScalarParameter() {defaultValue=0.05, group=Chest Light, parameterName=Day Opacity, sortPriority=41}
  n26 = ScalarParameter() {defaultValue=1, group=Chest Light, parameterName=Night Opacity, sortPriority=40}
  n27 = LinearInterpolate(A=@n25, B=@n26)
  n28 = MaterialFunctionCall(Value_Morning=@n27, Value_Day=@n25, Value_Evening=@n27, Value_Night=@n26) {materialFunction=&Scalar_Time_of_Day}
  n29 = TextureObject() {samplerType=SAMPLERTYPE_LinearGrayscale, texture=/Game/Global/Textures/Default/Flowmaps/T_StormSwirlMask}
  n30 = ScalarParameter() {defaultValue=1024, group=Chest Light, parameterName=Edge Warp Tiling, sortPriority=48}
  n31 = Multiply(A=Time(), B=ScalarParameter() {defaultValue=20, group=Chest Light, parameterName=Edge Warp Speed, sortPriority=45}) {constB=20}
  n32 = MaterialFunctionCall(
   TextureObject=@n29
   TextureSize=@n30
   WorldPosition=Add(
    A=WorldPosition() {worldPositionShaderOffset=WPT_ExcludeAllShaderOffsets}.XYZ
    B=MaterialFunctionCall(Z=@n31) {materialFunction=&MakeFloat3}.Result
   )
  ) {materialFunction=&WorldAlignedTexture}
  n33 = ScalarParameter() {defaultValue=200, group=Chest Light, parameterName=Edge Warp Amount, sortPriority=45}
  n34 = WorldPosition()
  n35 = Subtract(A=Add(A=Multiply(A=@n32.XYZ Texture, B=@n33) {constB=50}, B=@n34.XYZ), B=Divide(A=@n33))
  n36 = Multiply(
   A=Time()
   B=Multiply(A=ScalarParameter() {defaultValue=20, group=Chest Light, parameterName=Edge Warp Speed, sortPriority=45}) {constB=-0.66}
  ) {constB=20}
  n37 = MaterialFunctionCall(
   TextureObject=@n29
   TextureSize=@n30
   WorldPosition=Add(
    A=WorldPosition() {worldPositionShaderOffset=WPT_ExcludeAllShaderOffsets}.XYZ
    B=MaterialFunctionCall(X=@n36, Z=@n36) {materialFunction=&MakeFloat3}.Result
   )
  ) {materialFunction=&WorldAlignedTexture}
  n38 = Multiply(A=ScalarParameter() {defaultValue=200, group=Chest Light, parameterName=Edge Warp Amount, sortPriority=45}) {constB=-1}
  n39 = Subtract(A=Add(A=@n34.XYZ, B=Multiply(A=@n37.XYZ Texture, B=@n38) {constB=50}), B=Divide(A=@n38))
  n40 = MaterialProxyReplace(
   Realtime=Divide(A=Add(A=@n35, B=@n39))
   MaterialProxy=TransformPosition(None=WorldPosition().XYZ) {transformSourceType=TRANSFORMPOSSOURCE_World}
  )
  n41 = MaterialFunctionCall() {materialFunction=&ObjectScale}
  n42 = ScalarParameter() {defaultValue=215, group=Chest Light, parameterName=Inner Mask Edge Radius, sortPriority=28}
  n43 = Multiply(A=Time(), B=ScalarParameter() {defaultValue=1, group=Chest Light, parameterName=Flicker Speed, sortPriority=43})
  n44 = ConstantBiasScale(None=Sine(None=Multiply(A=@n43) {constB=0.66}))
  n45 = ConstantBiasScale(None=Sine(None=@n43))
  n46 = Sine(None=Multiply(A=@n43) {constB=0.45})
  n47 = Divide(A=Add(A=Add(A=@n44, B=@n45), B=ConstantBiasScale(None=@n46))) {constB=3}
  n48 = Multiply(
   A=SphereMask(
    A=@n12
    B=@n40
    Radius=Multiply(A=@n41.Scale X, B=@n42)
    Hardness=ScalarParameter() {defaultValue=0.16, group=Chest Light, parameterName=Inner Mask Edge Hardness, sortPriority=29}
   )
   B=LinearInterpolate(Alpha=@n47) {constA=0.7}
  )
  n49 = Multiply(
   A=@n41.Scale X
   B=MaterialProxyReplace(
    Realtime=ScalarParameter() {defaultValue=448, group=Chest Light, parameterName=Mask Radius, sortPriority=25}
    MaterialProxy=ScalarParameter() {defaultValue=448, group=Chest Light, parameterName=HLOD Mask Radius, sortPriority=25}
   )
  )
  n50 = LinearInterpolate(
   A=ScalarParameter() {group=Chest Light, parameterName=Mask Hardness Min, sortPriority=26}
   B=ScalarParameter() {defaultValue=0.05, group=Chest Light, parameterName=Mask Hardness Max, sortPriority=27}
   Alpha=@n47
  )
  n51 = ScalarParameter() {defaultValue=322, group=Chest Light, parameterName=Subtractive Mask Radius, sortPriority=30}
  n52 = Subtract(
   A=Saturate(None=SphereMask(A=@n12, B=@n40, Radius=@n49, Hardness=@n50))
   B=SphereMask(
    A=@n12
    B=@n40
    Radius=Multiply(A=@n41.Scale X, B=@n51)
    Hardness=ScalarParameter() {defaultValue=0.5, group=Chest Light, parameterName=Subtractive Mask Falloff, sortPriority=31}
   )
  )
  n53 = Clamp(None=Multiply(A=Max(A=@n48, B=@n52), B=MaterialProxyReplace(Realtime=@n28.Result, MaterialProxy=@n27)) {desc=Opacity})
  n54 = LinearInterpolate(
   A=@n2.EmissiveColor
   B=MaterialProxyReplace(Realtime=Multiply(A=@n22, B=@n23) {desc=Emissive}, MaterialProxy=Multiply(A=@n22, B=@n24) {desc=Emissive})
   Alpha=Multiply(A=@n53, B=@n2.AmbientOcclusion)
  )
  n55 = ScalarParameter() {group=Chest Light, parameterName=chain_min_value}
  n56 = ScalarParameter() {defaultValue=1, group=Chest Light, parameterName=chain_max_value}
  n57 = StaticSwitchParameter(True=Multiply(A=@n54, B=LinearInterpolate(A=@n55, B=@n56, Alpha=@n2.Metallic) {constA=1}), False=@n54) {group=Chest Light, parameterName=Chain?}
  FunctionOutput(None=StaticSwitchParameter(
    True=SetMaterialAttributes!(MaterialAttributes=@n2.MaterialAttributes, BaseColor=@n2.MaterialAttributes, EmissiveColor=@n57) {attributeSetTypes [#1, #2]}
    False=@n1
   ) {group=Chest Light, parameterName=UseChestFakedLight, sortPriority=1}) {outputName=Material}
  FunctionOutput(None=@n53) {outputName=Mask, sortPriority=1}
  StaticSwitchParameter(
   True=LinearInterpolate(A=@n2.BaseColor, B=@n22, Alpha=Multiply(A=@n2.AmbientOcclusion, B=@n53))
   False=LinearInterpolate(A=@n2.BaseColor, B=@n22, Alpha=@n53)
  ) {group=Chest Light, parameterName=Chain?}
 ]}
