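# Material function for the Central Picnic hero rock props, building a layered rock surface with faked lighting.
# Projects tiling rock in pre-skinned space, then shapes it with UV1 masks and normals, detail maps and an optional custom top layer.
# AO, color tint, bounds-relative gradient and dirt (skipped by the hand switch) follow, then chest-orb light and optional Vortex_Light emissive sphere glow.
# Shading-path, quality and Nanite switches vary cost and normal strength.
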MaterialFunction /Game/Environments/Helios/Props/CentralPicnic/Materials/MF_Central_Picnic_HeroRock
graph {nodes [
  n1 = ScalarParameter() {defaultValue=1, group=B_L2, parameterName=L2 Scale Details, sortPriority=2}
  n2 = Divide(A=ObjectRadius()) {constB=256}
  n3 = NamedRerouteDeclaration(None=Multiply(A=TextureCoordinate(), B=Multiply(A=@n1, B=@n2))) {name=L2 Uvs, nodeColor=<rgba 0.9922 1 0 1>}
  n4 = TextureSampleParameter2D(UVs=@n3) {
   group=B_L2
   parameterName=Normals_Texture_2
   samplerSource=SSM_Wrap_WorldGroupSettings
   samplerType=SAMPLERTYPE_Normal
   sortPriority=2
   texture=/Engine/EngineMaterials/FlatNormal
  }
  n5 = Multiply(A=ComponentMask(None=@n4.RGB) {g=true, r=true}, B=ScalarParameter() {defaultValue=1, group=B_L2, parameterName=Details Normals Value})
  n6 = VectorParameter() {
   group=B_L2
   parameterName="AO ( Expand for blending option)"
   channelNames {a=Linear vs Mult, b="", g="", r=""}
   defaultValue=<rgba 0.2691 0.1783 0.1117 0.5>
  }
  n7 = Constant() {r=1}
  n8 = Multiply(A=ScalarParameter() {defaultValue=4, group=B_L2, parameterName=Edge Mask Multiplier}, B=@n7) {constA=4}
  n9 = TextureCoordinate() {coordinateIndex=1}
  n10 = TextureSampleParameter2D(UVs=@n9) {
   group=B_L2
   parameterName=SpecularMasks_2
   samplerSource=SSM_Wrap_WorldGroupSettings
   samplerType=SAMPLERTYPE_Masks
   sortPriority=1
   texture=&T_Fortnite_Default_S
  }
  n11 = TextureObjectParameter() {
   group=A_L1_Rock Surface
   parameterName=Specular_Tile
   samplerType=SAMPLERTYPE_Masks
   texture=&T_Fortnite_Default_S
  }
  n12 = MaterialFunctionCall!(
   Diffuse Texture Object=TextureObjectParameter() {group=A_L1_Rock Surface, parameterName=Diffuse_Tile, texture=/Game/Global/Textures/Default/Blanks/T_MidgreysRGB}
   Spec Texture Object=@n11
   TextureSize=ScalarParameter() {defaultValue=1, group=A_L1_Rock Surface, parameterName=L1 Tile, sortPriority=5}
   --------------------------
   Use seperate settings for Z projection
   Use seperate texture for Z projection
   Texture Size Z projection
   Z Plane Diffuse Texture Object
   Z Plane Spec Texture Object
   "--------------------------      "
   ProjectionTransitionContrast
   Use Texture Alpha Heightmap to Refine the Transistion
   HeightLerpTransitionContrast
   PickTheDiffuseHeightChannel
   TheDiffuseHeightChannel
   "--------------------------       "
   NormalTextureObject
   World Space Normals
   "--------------------------  "
   Export Float 4 Diffuse
   Export Float 4 Spec
   "--------------------------     "
   WorldPosition=VertexInterpolator(VS=PreSkinnedPosition()).PS
   World Space Normal=VertexInterpolator(VS=PreSkinnedNormal()).PS
  ) {materialFunction=/Engine/Functions/Engine_MaterialFunctions02/Texturing/WorldAlignedTextures_Complex}
  n13 = Desaturation(None=@n12.XYZ Diffuse Texture, Fraction=ScalarParameter() {group=A_L1_Rock Surface, parameterName=L1 Desaturate})
  MaterialExpressionReroute_221 = Reroute(None=@n13)
  n14 = VectorParameter() {group=A_L1_Rock Surface, parameterName=L1 Tint, channelNames {a=Blend to color, b="", g="", r=""}, defaultValue=<rgba 1 1 1 0>}
  n15 = ShadingPathSwitch(Default=@n14, Mobile=Constant3Vector() {constant=<rgba 5 5 5 0>})
  n16 = Desaturation(
   None=@MaterialExpressionReroute_221.A
   Fraction=ScalarParameter() {defaultValue=0.8, group=A_L1_Rock Surface, parameterName=L1_Cliff_Moss_Mountains_DesaturationAmt}
  )
  n17 = VectorParameter() {group=A_L1_Rock Surface, parameterName=L1 Tint Green, channelNames {a=Blend to color, b="", g="", r=""}, defaultValue=<rgba 1 1 1 0>}
  n18 = MaterialFunctionCall(Float3=@n12.XYZ Spec) {materialFunction=&BreakOutFloat3Components}
  MaterialExpressionReroute_222 = Reroute(None=@n18.G)
  n19 = LinearInterpolate(
   A=LinearInterpolate(A=Multiply(A=@MaterialExpressionReroute_221.A, B=@n15), B=@n15, Alpha=@n14.Blend to color)
   B=LinearInterpolate(A=Multiply(A=@n16, B=@n17), B=@n17, Alpha=@n17)
   Alpha=@MaterialExpressionReroute_222.Alpha
  )
  MaterialExpressionReroute_223 = Reroute(None=@n19)
  n20 = ScalarParameter() {defaultValue=0.5, group=A_L1_Rock Surface, parameterName="SpecMoss ( Replaces AO"}
  n21 = ScalarParameter() {group=A_L1_Rock Surface, parameterName=Roughness_Min}
  n22 = ScalarParameter() {defaultValue=1, group=A_L1_Rock Surface, parameterName=Roughness_Max}
  n23 = Multiply(
   A=Power(Base=@n18.B, Exp=ScalarParameter() {defaultValue=2, group=A_L1_Rock Surface, parameterName=Roughness_Pow})
   B=ScalarParameter() {defaultValue=1, group=A_L1_Rock Surface, parameterName=Roughness_Multi}
  )
  n24 = TextureCoordinate()
  n25 = TextureSampleParameter2D(UVs=@n24) {
   group=A_L1_Rock Surface
   parameterName=Normals
   samplerSource=SSM_Wrap_WorldGroupSettings
   samplerType=SAMPLERTYPE_Normal
   sortPriority=2
   texture=&FlatNormal
  }
  n26 = ScalarParameter() {defaultValue=1.2, group=A_L1_Rock Surface, parameterName="L1 Normals Value Baked Prop (non-Nanite)", sortPriority=5}
  n27 = ScalarParameter() {defaultValue=2, group=A_L1_Rock Surface, parameterName="L1 Normals Value (non-Nanite)", sortPriority=5}
  n28 = StaticBoolParameter() {group=B_L2, parameterName=Baked Prop UV2, sortPriority=0}
  n29 = ScalarParameter() {defaultValue=1, group=A_L1_Rock Surface, parameterName="L1 Normals Value (Nanite)", sortPriority=5}
  n30 = Multiply(
   A=NamedRerouteDeclaration(None=StaticSwitch(True=@n26, False=@n27, Value=@n28)) {name=Normal Intensity Mult, nodeColor=<rgba 0 0.8594 1 1>}
   B=@n29
  ) {constA=2}
  n31 = AppendVector(A=Multiply(A=ComponentMask(None=@n25.RGB) {g=true, r=true}, B=NaniteReplace!(Default=@n30, Nanite=@n29)), B=@n25.B)
  n32 = MaterialFunctionCall(Float4=@n12.XYZ Diffuse Texture) {materialFunction=&BreakOutFloat4Components}
  n33 = SetMaterialAttributes!(
   MaterialAttributes
   BaseColor=@MaterialExpressionReroute_223.Output
   Specular=NamedRerouteDeclaration(None=LinearInterpolate(A=@n18.R, B=@n20, Alpha=@n18.R)) {name=L1 Specular, nodeColor=<rgba 1 0 0.1875 1>}
   Roughness=Saturate(None=LinearInterpolate(A=@n21, B=@n22, Alpha=@n23))
   Normal=NamedRerouteDeclaration(None=Normalize(VectorInput=@n31)) {name=L1 Normals, nodeColor=<rgba 0.8984 0 1 1>}
   Opacity=@n32.A
  ) {attributeSetTypes [
    #1
    #2
    #3
    #4
    #5
   ]}
  n34 = NamedRerouteDeclaration(None=@n33) {name=Mobile Rock Material, nodeColor=<rgba 1 0 0.5859 1>}
  n35 = GetMaterialAttributes(None=@n34) {attributeGetTypes [
    #1
    #2
    #3
    #4
    #5
   ]}
  n36 = NamedRerouteDeclaration(None=MaterialFunctionCall!(Linear=Multiply(A=@n35.Opacity, B=@n10.R)) {materialFunction=/Engine/Functions/Engine_MaterialFunctions01/ImageAdjustment/LinearTosRGB}.sRGB) {name=Height L1, nodeColor=<rgba 0.4766 1 0 1>}
  n37 = Multiply(A=Multiply(A=LinearInterpolate(A=@n7, B=@n8, Alpha=@n10.G), B=@n35.BaseColor), B=@n36)
  n38 = Multiply(A=@n6, B=@n37)
  n39 = Multiply(A=LinearInterpolate(A=@n35.Specular, Alpha=NamedRerouteUsage()), B=@n10.R)
  n40 = ScalarParameter() {defaultValue=0.75, group=A_L1_Rock Surface, parameterName=R Deeps}
  n41 = ScalarParameter() {defaultValue=2, group=A_L1_Rock Surface, parameterName=R Edges}
  n42 = TextureSampleParameter2D(UVs=@n9) {
   group=B_L2
   parameterName=Normals_Texture_2
   samplerSource=SSM_Wrap_WorldGroupSettings
   samplerType=SAMPLERTYPE_Normal
   sortPriority=2
   texture=&FlatNormal
  }
  n43 = Transform(None=ShaderStageSwitch(
    PixelShader=MaterialFunctionCall!(UVs=@n9, Normal=@n42.RGB) {materialFunction=/Engine/Functions/Engine_MaterialFunctions01/Texturing/DeriveTangentBasis}.World Space Normal
    VertexShader=@n42.RGB
   )) {transformSourceType=TRANSFORMSOURCE_World, transformType=TRANSFORM_Tangent}
  n44 = StaticSwitch(
   True=ScalarParameter() {defaultValue=1.2, group=B_L2, parameterName="L2 Normals Value Baked Prop (non-Nanite)", sortPriority=5}
   False=ScalarParameter() {defaultValue=2, group=B_L2, parameterName="L2 Normals Value (non-Nanite)", sortPriority=5}
   Value=StaticBoolParameter() {group=B_L2, parameterName=Baked Prop UV2, sortPriority=0}
  )
  n45 = ScalarParameter() {defaultValue=1, group=B_L2, parameterName="L2 Normals Value (Nanite)", sortPriority=5}
  n46 = Multiply(A=ComponentMask(None=@n43) {g=true, r=true}, B=NaniteReplace(Default=Multiply(A=@n44, B=@n45) {constA=2}, Nanite=@n45))
  n47 = ComponentMask(None=@n43) {b=true}
  n48 = Constant3Vector() {constant=<rgba 1 1 0 0>}
  n49 = SetMaterialAttributes!(
   MaterialAttributes
   BaseColor=LinearInterpolate(A=LinearInterpolate(A=@n38, B=@n6, Alpha=@n6), B=@n37, Alpha=Saturate(None=@n39)) {constA=1}
   Specular=Multiply(A=@n35.Specular, B=@n36)
   Roughness=Multiply(A=LinearInterpolate(A=@n40, B=@n41, Alpha=@n36), B=@n35.Roughness)
   Normal=Add(A=AppendVector(A=@n46, B=@n47), B=Multiply(A=@n35.Normal, B=@n48))
  ) {attributeSetTypes [
    #1
    #2
    #3
    #4
   ]}
  n50 = GetMaterialAttributes(None=@n49) {attributeGetTypes [#1, #3]}
  n51 = Add(B=@n50.BaseColor) {constA=0.5}
  n52 = NamedRerouteUsage()
  n53 = OneMinus(None=@n50.Roughness)
  n54 = Saturate(None=LinearInterpolate(A=Constant() {r=-2.5}, B=Constant() {r=2.5}, Alpha=Subtract(A=@n52, B=@n53)))
  n55 = Saturate(None=LinearInterpolate(A=Multiply(A=@n51, B=@n50.BaseColor), B=@n50.BaseColor, Alpha=@n54))
  n56 = LinearInterpolate(A=Multiply(A=@n50.Roughness) {constB=2}, B=@n50.Roughness, Alpha=@n54)
  n57 = ShadingPathSwitch(
   Default=StaticSwitch(
    True=SetMaterialAttributes!(MaterialAttributes=@n50.MaterialAttributes, BaseColor=@n55, Roughness=@n56) {attributeSetTypes [#1, #3]}
    False=NamedRerouteUsage()
    Value=StaticBoolParameter() {group=B_L2, parameterName=Baked Prop UV2, sortPriority=0}
   )
   Mobile=NamedRerouteUsage()
  )
  n58 = StaticSwitchParameter(True=BlendMaterialAttributes(A=@n57, B=NamedRerouteUsage(), Alpha=NamedRerouteUsage()), False=@n57) {desc="If set to True, set TopTerrainMaterial and TopSnow to False", group=C_L3 Top, parameterName=Custom Top Layer, sortPriority=-1}
  n59 = NamedRerouteDeclaration(None=@n58) {name=Rocks Custom Top Blended, nodeColor=<rgba 0 1 0.4609 1>}
  n60 = GetMaterialAttributes(None=@n59) {attributeGetTypes [#1, #4, #2]}
  Add(A=@n60.Normal, B=@n5)
  Composite!(A=@n13, Alpha=@n18.G) {subgraphName=Color tint}
  Constant3Vector() {constant=<rgba 1 1 0 0>}
  Constant() {r=1}
  Constant() {r=1}
  n61 = TextureSampleParameter2D(UVs=@n3) {
   group=B_L2
   parameterName=SpecularMasks_2
   samplerSource=SSM_Wrap_WorldGroupSettings
   samplerType=SAMPLERTYPE_Masks
   sortPriority=1
   texture=&T_Fortnite_Default_S
  }
  n62 = Multiply(A=@n61.R, B=ScalarParameter() {defaultValue=1, group=B_L2, parameterName=Detail Brightess})
  n63 = ScalarParameter() {group=B_L2, parameterName=Specular_Color_Influence}
  n64 = StaticSwitchParameter(True=@n60.BaseColor, False=Multiply(A=@n60.BaseColor, B=LinearInterpolate(B=@n62, Alpha=@n63) {constA=1})) {desc="If set to True, set TopTerrainMaterial and TopSnow to False", group=C_L3 Top, parameterName=Custom Top Layer, sortPriority=-1}
  n65 = AppendVector(A=@n5, B=@n4.B)
  n66 = Power(Base=@n61.R, Exp=ScalarParameter() {defaultValue=2, group=B_L2, parameterName=Specular Pow})
  n67 = ScalarParameter() {defaultValue=1, group=B_L2, parameterName=Specular Brightess}
  n68 = Multiply(A=@n61.R, B=LinearInterpolate(B=Multiply(A=@n66, B=@n67), Alpha=ScalarParameter() {group=B_L2, parameterName=Specular_Influence}) {constA=1})
  n69 = StaticSwitch(
   True=@n59
   False=SetMaterialAttributes!(
    MaterialAttributes=@n60.MaterialAttributes
    BaseColor=@n64
    Normal=MaterialFunctionCall(BaseNormal=@n60.Normal, AdditionalNormal=@n65) {materialFunction=&BlendAngleCorrectedNormals}.Result
    Specular=Clamp(None=@n68) {minDefault=0.2}
   ) {attributeSetTypes [#1, #4, #2]}
   Value=StaticBoolParameter() {group=B_L2, parameterName=Baked Prop UV2, sortPriority=0}
  )
  n70 = MaterialFunctionCall!(
   UV's=NamedRerouteUsage()
   Checker true=StaticBoolParameter() {group=B_L2, parameterName=L2 Details Preview UVs}
   Material In=ShadingPathSwitch(Default=StaticSwitchParameter(True=@n69, False=@n59) {group=B_L2, parameterName=L2 Details, sortPriority=-1}, Mobile=@n59)
  ) {materialFunction=/Game/Global/Base4Layers_2022/MaterialFunctions/MF_Checker}
  n71 = GetMaterialAttributes(None=QualitySwitch(Default=NamedRerouteUsage(), High=@n70.Switch, Epic=@n70.Switch)) {attributeGetTypes [#4]}
  n72 = SetMaterialAttributes!(MaterialAttributes=@n71.MaterialAttributes, Normal=MaterialProxyReplace(Realtime=Transform(None=@n71.Normal), MaterialProxy=@n71.Normal)) {attributeSetTypes [#4]}
  n73 = GetMaterialAttributes(None=@n72) {attributeGetTypes [#1]}
  n74 = TextureSampleParameter2D() {
   group=A_L1_Rock Surface
   parameterName=Rock_AO
   texture=/Game/Environments/Helios/Props/CentralPicnic/Textures/T_CentralPicnic_Arm_HandAndSpike_AO
  }
  n75 = LinearInterpolate(
   A=Multiply(A=@n73.BaseColor, B=ScalarParameter() {defaultValue=1, group=A_L1_Rock Surface, parameterName=AO_Darken})
   B=@n73.BaseColor
   Alpha=@n74.RGB
  )
  n76 = TextureSampleParameter2D(UVs=Multiply(A=TextureCoordinate(), B=ScalarParameter() {defaultValue=1, group=Color_Adjustment, parameterName=Color texture scale, sortPriority=98})) {
   group=Color_Adjustment
   parameterName=Color_Texture
   samplerSource=SSM_Wrap_WorldGroupSettings
   texture=/Game/Environments/Apollo/Rocks/Natural_Rocks/Textures/T_RockDetails_02D
  }
  n77 = LinearInterpolate(
   A=@n75
   B=Multiply(A=@n75, B=VectorParameter() {group=Color_Adjustment, parameterName=Tint1, defaultValue=<rgba 1 1 1 1>}.RGB)
   Alpha=Multiply(A=@n76.RGB, B=ScalarParameter() {group=Color_Adjustment, parameterName=Color_Brightness})
  )
  n78 = VectorParameter() {group=A_L1_Rock Surface, parameterName=gradient_Multi, defaultValue=<rgba 1 1 1 1>}
  n79 = MaterialProxyReplace(Realtime=@n78.RGB, MaterialProxy=Multiply(A=@n78.RGB) {constB=0.8})
  n80 = WorldPosition()
  n81 = ObjectPositionWS()
  n82 = Divide(
   A=Subtract(
    A=TransformPosition(None=@n80.XYZ) {transformSourceType=TRANSFORMPOSSOURCE_World}
    B=TransformPosition(None=@n81) {transformSourceType=TRANSFORMPOSSOURCE_World}
   )
   B=VertexInterpolator(VS=PreSkinnedLocalBounds().Extents).PS
  )
  n83 = Add(A=@n82, B=ScalarParameter() {defaultValue=0.4472, group=A_L1_Rock Surface, parameterName=gradient offset, sliderMax=1, sortPriority=33}) {constB=0.5}
  n84 = ScalarParameter() {defaultValue=2.8996, group=A_L1_Rock Surface, parameterName=gradientPow, sortPriority=33}
  n85 = Clamp(None=ComponentMask(None=Power(Base=@n83, Exp=@n84)) {g=true, r=true})
  n86 = NamedRerouteUsage()
  n87 = SetMaterialAttributes!(
   MaterialAttributes=@n72
   BaseColor=LinearInterpolate(A=Multiply(A=@n77, B=@n79), B=@n77, Alpha=MaterialProxyReplace(Realtime=@n85, MaterialProxy=@n86))
  ) {attributeSetTypes [#1]}
  n88 = VectorParameter() {
   group=C_Dirt_Layer
   parameterName=Dirt Color
   sortPriority=0
   channelNames {a=Dirt Global Transparency, b="", g="", r=""}
   defaultValue=<rgba 0.4149 0.2605 0.1379 1>
  }
  n89 = SetMaterialAttributes!(
   MaterialAttributes=@n87
   BaseColor=@n88
   Roughness=ScalarParameter() {defaultValue=0.75, group=C_Dirt_Layer, parameterName=Dirt_Roughness}
   Specular=ScalarParameter() {defaultValue=0.5, group=C_Dirt_Layer, parameterName=Dirt_Spec}
   Metallic=Constant()
  ) {attributeSetTypes [
    #1
    #3
    #2
    #6
   ]}
  n90 = ScalarParameter() {defaultValue=0.4472, group=C_Dirt_Layer, parameterName=dirt offset, sortPriority=33}
  n91 = ComponentMask(None=Power(
    Base=Add(A=@n82, B=@n90) {constB=0.5}
    Exp=ScalarParameter() {defaultValue=2.8996, group=C_Dirt_Layer, parameterName=dirtPow, sortPriority=33}
   )) {b=true}
  n92 = NamedRerouteUsage()
  n93 = Power(
   Base=ComponentMask(None=GetMaterialAttributes(None=@n92) {attributeGetTypes [#1]}.BaseColor) {r=true}
   Exp=ShadingPathSwitch(Default=ScalarParameter() {defaultValue=2, group=C_Dirt_Layer, parameterName=Dirt_subtract_Pow}, Mobile=Constant() {r=2.6856})
  )
  n94 = ScalarParameter() {defaultValue=1, group=C_Dirt_Layer, parameterName=Dirt_subtract_Multi}
  n95 = Clamp(None=Subtract(A=@n91, B=Multiply(A=@n93, B=@n94)))
  n96 = GetMaterialAttributes(None=StaticSwitchParameter(True=@n87, False=BlendMaterialAttributes(A=@n89, B=@n87, Alpha=@n95) {vertexAttributeBlendType=UseB}) {group=Controls, parameterName=hand?}) {attributeGetTypes [#2]}
  n97 = Multiply(A=@n96.Specular, B=ScalarParameter() {defaultValue=1, group=A_L1_Rock Surface, parameterName=AO_Darken_Spec})
  n98 = TextureSampleParameter2D() {
   group=A_L1_Rock Surface
   parameterName=Rock_AO
   texture=&T_CentralPicnic_Arm_HandAndSpike_AO
  }
  n99 = SetMaterialAttributes!(MaterialAttributes=@n96.MaterialAttributes, Specular=Clamp(None=LinearInterpolate(A=@n97, B=@n96.Specular, Alpha=@n98.RGB)) {minDefault=0.2}) {attributeSetTypes [#2]}
  n100 = GetMaterialAttributes(None=@n99) {attributeGetTypes [#4]}
  n101 = MaterialFunctionCall!(Material Attributes=@n99, World Normal Map=@n100.Normal, Vertex Normals, Use Vertex Normal Masking) {materialFunction=&/MF_CentralPicnic_ChestOrbFakedLight}
  n102 = Add(
   A=Transform(None=VectorParameter() {group=Top light, parameterName=CenterPoint2, sortPriority=10, defaultValue=<rgba 0 0 0 1>}.RGB) {transformSourceType=TRANSFORMSOURCE_Local}
   B=MaterialFunctionCall() {materialFunction=&ComponentPivotLocation}.Result
  )
  n103 = SphereMask(
   A=@n102
   B=WorldPosition() {worldPositionShaderOffset=WPT_ExcludeAllShaderOffsets}.XYZ
   Radius=Multiply(
    A=MaterialFunctionCall() {materialFunction=&ObjectScale}.Scale X
    B=ScalarParameter() {defaultValue=448, group=Top light, parameterName=Mask Radius B, sortPriority=25}
   )
   Hardness=ScalarParameter() {defaultValue=50, group=Top light, parameterName=Mask_B_Contrast}
  )
  n104 = Normalize(VectorInput=@n100.Normal)
  n105 = DotProduct(A=Normalize(VectorInput=CameraVectorWS()), B=@n104)
  n106 = Multiply(
   A=Power(Base=Saturate(None=@n105), Exp=ScalarParameter() {defaultValue=1, group=Top light, parameterName=Fresnel Contrast B, sortPriority=36})
   B=ScalarParameter() {group=Top light, parameterName=Fresnel Brightness B, sortPriority=37}
  )
  n107 = Subtract(A=@n102, B=WorldPosition().XYZ)
  n108 = Saturate(None=DotProduct(A=Normalize(VectorInput=@n107), B=@n104))
  n109 = Normalize(VectorInput=Subtract(A=@n102, B=WorldPosition().XYZ))
  n110 = Normalize(VectorInput=VertexNormalWS())
  n111 = LinearInterpolate(B=Saturate(None=DotProduct(A=@n109, B=@n110)), Alpha=StaticSwitch(True=Constant() {r=1}, False=Constant(), Value=StaticBool() {value=true})) {constA=1}
  n112 = Multiply(
   A=Add(A=Saturate(None=@n106), B=LinearInterpolate(B=@n108, Alpha=@n111))
   B=VectorParameter() {group=Top light, parameterName=Color B, sortPriority=20, defaultValue=<rgba 1 0.263 0.032 1>}.RGB
  ) {desc=Color}
  n113 = ScalarParameter() {group=Top light, parameterName=light intensity B}
  n114 = Add(
   A=GetMaterialAttributes(None=@n101.Material) {attributeGetTypes [#7]}.EmissiveColor
   B=Multiply(A=Saturate(None=@n103), B=Multiply(A=@n112, B=@n113))
  )
  FunctionOutput(None=StaticSwitchParameter(
    True=SetMaterialAttributes!(MaterialAttributes=@n101.Material, EmissiveColor=@n114) {attributeSetTypes [#7]}
    False=@n101.Material
   ) {parameterName=Vortex_Light})
  MaterialFunctionCall!(Material, Tint=@n88, TintMask=@n95) {materialFunction=/Engine/Functions/MaterialLayerFunctions/MatLayerBlend_Tint}
  MaterialProxyReplace(MaterialProxy=NamedRerouteUsage())
  Multiply(A=@n75, B=VectorParameter() {group=Color_Adjustment, parameterName=Tint2, defaultValue=<rgba 1 1 1 1>}.RGB)
  n115 = GetMaterialAttributes(None=@n34) {attributeGetTypes [#1, #2]}
  n116 = TextureObjectParameter() {group=Cliff, parameterName=Cliff_DetailTexture, texture=&T_RockDetails_02D}
  n117 = MaterialFunctionCall(
   Diffuse Texture Object=@n116
   TextureSize=ScalarParameter() {defaultValue=700, group=Cliff, parameterName=Cliff_DetailTexture_Scale, sortPriority=98}
   ProjectionTransitionContrast=Constant() {r=5}
   WorldPosition=VertexInterpolator(VS=PreSkinnedPosition()).PS
   World Space Normal=VertexInterpolator(VS=PreSkinnedNormal()).PS
  ) {materialFunction=&WorldAlignedTextures_Complex}
  n118 = Multiply(
   A=@n115.BaseColor
   B=LinearInterpolate(
    A=ScalarParameter() {group=A_L1_Rock Surface, parameterName=Detail_Bright_Value}
    B=ScalarParameter() {defaultValue=1, group=A_L1_Rock Surface, parameterName=Detail_Dark_Value}
    Alpha=@n117.XYZ Diffuse Texture
   )
  )
  Multiply(A=@n118, B=Constant() {r=1.5})
  n119 = OneMinus(None=DotProduct(A=TextureCoordinate(), B=Constant2Vector() {g=1}))
  n120 = ScalarParameter() {group=A_L1_Rock Surface, parameterName=HLOD_Purple_Offset}
  NamedRerouteDeclaration(None=Saturate(None=Subtract(A=@n119, B=@n120))) {name=hlod_purple gradient, nodeColor=<rgba 0 0.5078 1 1>}
  n121 = OneMinus(None=DotProduct(A=TextureCoordinate(), B=Constant2Vector() {g=1}))
  n122 = ScalarParameter() {group=C_Dirt_Layer, parameterName=HLOD_Dirt_Offset}
  NamedRerouteDeclaration(None=Saturate(None=Subtract(A=@n121, B=@n122))) {name=hlod_dirt_Gradient, nodeColor=<rgba 0 0.5078 1 1>}
  n123 = VectorParameter() {group=D_TopMask, parameterName=VertexColor Mask for Grass, sortPriority=1, defaultValue=<rgba 0 0 0 1>}
  n124 = VertexColor()
  n125 = DotProduct(A=@n123.RGB, B=AppendVector(A=@n124, B=@n124))
  n126 = StaticSwitchParameter(True=@n125, False=OneMinus(None=@n125)) {group=D_TopMask, parameterName=Invert Vertex mask, sortPriority=2}
  n127 = GetMaterialAttributes(None=@n57) {attributeGetTypes [#4]}
  n128 = SetMaterialAttributes!(MaterialAttributes=@n127.MaterialAttributes, Normal=Transform(None=@n127.Normal)) {attributeSetTypes [#4]}
  n129 = Saturate(None=ComponentMask(None=GetMaterialAttributes(None=@n128) {attributeGetTypes [#4]}.Normal) {b=true})
  n130 = ScalarParameter() {defaultValue=1, group=D_TopMask, parameterName=TopMask_TextureNormalZ_Power, sortPriority=6}
  n131 = Saturate(None=Multiply(
    A=Power(Base=@n129, Exp=@n130)
    B=ScalarParameter() {defaultValue=1, group=D_TopMask, parameterName=TopMask_TextureNormalZ_Mod, sortPriority=6}
   ))
  n132 = StaticSwitchParameter(True=Max(A=@n126, B=@n131), False=@n126) {group=D_TopMask, parameterName="Add Top Masks (Pixel on Vertex)"}
  NamedRerouteDeclaration(None=Saturate(None=StaticSwitchParameter(True=@n132, False=@n131) {group=D_TopMask, parameterName=Use Top Mask Vertex Colors})) {name=CustomTopLayer Blend Mask, nodeColor=<rgba 1 0.3323 0 1>}
  n133 = Subtract(A=WorldPosition() {worldPositionShaderOffset=WPT_ExcludeAllShaderOffsets}.XYZ, B=ActorPositionWS())
  n134 = MaterialFunctionCall() {materialFunction=&ObjectLocalBounds}
  n135 = ComponentMask(None=Divide(A=Subtract(A=@n133, B=@n134.Local Bounds Minimum), B=@n134.Local Bounds Size)) {b=true}
  n136 = ComponentMask(None=VertexNormalWS()) {b=true}
  NamedRerouteDeclaration(None=Saturate(None=Max(A=@n135, B=@n136))) {name=Fade From Center To Top, nodeColor=<rgba 0 0.1562 1 1>}
  n137 = GetMaterialAttributes(None=@n57) {attributeGetTypes [#4]}
  n138 = SetMaterialAttributes!(
   MaterialAttributes=@n137.MaterialAttributes
   BaseColor=NamedRerouteUsage()
   Specular=NamedRerouteUsage()
   Roughness=NamedRerouteUsage()
   Normal=NamedRerouteUsage()
  ) {attributeSetTypes [
    #1
    #2
    #3
    #4
   ]}
  NamedRerouteDeclaration(None=@n138) {name=Custom Top MA, nodeColor=<rgba 1 0.3044 0 1>}
  n139 = StaticBoolParameter() {desc="True - Baked Props Will Use WS projection Instead \r\nFalse - Baked Props Use UV Map", group=C_L3 Top, parameterName=Use WS}
  n140 = ObjectRadius()
  n141 = Multiply(
   A=Multiply(A=ScalarParameter() {defaultValue=1, group=C_L3 Top, parameterName=L3 Tile, sortPriority=5}, B=Divide(A=@n140) {constB=256})
   B=TextureCoordinate()
  )
  n142 = TextureSampleParameter2D(UVs=@n141) {
   group=C_L3 Top
   parameterName=Diffuse_Texture_3
   samplerSource=SSM_Wrap_WorldGroupSettings
   sortPriority=0
   texture=&T_MidgreysRGB
  }
  n143 = StaticBoolParameter() {
   desc="True - Baked Props Will Use XYZ Projection \r\nFalse - Baked Props use only Z Projection for Custom Top"
   group=C_L3 Top
   parameterName=Use XYZ
  }
  n144 = TextureObjectParameter() {
   group=C_L3 Top
   parameterName=SpecularMasks_3
   samplerType=SAMPLERTYPE_Masks
   texture=&T_Fortnite_Default_S
  }
  n145 = TextureObjectParameter() {group=C_L3 Top, parameterName=Normals_Texture_3, samplerType=SAMPLERTYPE_Normal, texture=&FlatNormal}
  n146 = MaterialFunctionCall(
   Diffuse Texture Object=TextureObjectParameter() {group=C_L3 Top, parameterName=Diffuse_Texture_3, texture=&T_MidgreysRGB}
   Spec Texture Object=@n144
   TextureSize=ScalarParameter() {defaultValue=2000, group=C_L3 Top, parameterName=L3 WS Tile, sortPriority=5}
   NormalTextureObject=@n145
   World Space Normals=StaticBool()
   Export Float 4 Diffuse=StaticBool() {value=true}
  ) {materialFunction=&WorldAlignedTextures_Complex}
  n147 = ShadingPathSwitch(Default=StaticSwitch(True=@n146.XYZ Diffuse Texture, False=@n146.Z Diffuse Texture, Value=@n143), Mobile=@n142.RGBA)
  n148 = StaticSwitchParameter(True=StaticSwitch(True=@n147, False=@n142.RGBA, Value=@n139), False=@n147) {group=B_L2, parameterName=Baked Prop UV2, sortPriority=0}
  NamedRerouteDeclaration(None=MaterialFunctionCall(
    Transition Phase=NamedRerouteUsage()
    Height Texture=ComponentMask(None=@n148) {a=true}
    Contrast=ScalarParameter() {group=D_TopMask, parameterName=Top Contrast, sortPriority=6}
   ) {materialFunction=&HeightLerp}.Alpha) {name=Custom Top H, nodeColor=<rgba 0.4531 1 0 1>}
  n149 = ComponentMask(None=@n117.XYZ Diffuse Texture) {r=true}
  n150 = Add(A=Constant() {r=-1.1}, B=Multiply(A=Constant() {r=2.05}, B=@n149))
  n151 = LinearInterpolate(A=Constant() {r=7}, B=Constant() {r=-0.8}, Alpha=@n149)
  n152 = LinearInterpolate(
   B=Multiply(A=Saturate(None=@n150), B=Saturate(None=@n151))
   Alpha=ScalarParameter() {defaultValue=1, group=B_L2, parameterName=Landscape Noise Amount_Spec}
  ) {constA=1}
  n153 = Constant() {r=-0.25}
  n154 = Constant() {r=1.5}
  n155 = Clamp(None=Saturate(None=LinearInterpolate(A=@n153, B=@n154, Alpha=@n115.Specular)), Max=Constant() {r=0.5}) {clampMode=CMODE_ClampMax}
  NamedRerouteDeclaration(None=SetMaterialAttributes!(MaterialAttributes=@n115.MaterialAttributes, BaseColor=@n118, Specular=Multiply(A=@n152, B=@n155)) {attributeSetTypes [#1, #2]}) {name=Cliff Details MA, nodeColor=<rgba 0.4297 0 1 1>}
  n156 = TextureSampleParameter2D(UVs=@n141) {
   group=C_L3 Top
   parameterName=Normals_Texture_3
   samplerSource=SSM_Wrap_WorldGroupSettings
   samplerType=SAMPLERTYPE_Normal
   sortPriority=2
   texture=&FlatNormal
  }
  n157 = ShadingPathSwitch(Default=StaticSwitch(True=@n146.Normal XYZ Texture, False=@n146.Normal Z Texture, Value=@n143), Mobile=@n156.RGB)
  n158 = StaticSwitchParameter(True=StaticSwitch(True=@n157, False=@n156.RGB, Value=@n139), False=@n157) {group=B_L2, parameterName=Baked Prop UV2, sortPriority=0}
  n159 = AppendVector(
   A=Multiply(
    A=ScalarParameter() {defaultValue=1, group=C_L3 Top, parameterName=L3 Normals Value, sortPriority=5}
    B=ComponentMask(None=@n158) {g=true, r=true}
   )
   B=ComponentMask(None=@n158) {b=true}
  )
  n160 = Constant3Vector() {constant=<rgba 1 1 0 0>}
  n161 = LinearInterpolate(
   A=@n159
   B=Add(A=Multiply(A=@n137.Normal, B=@n160), B=@n159)
   Alpha=Saturate(None=ScalarParameter() {defaultValue=1, group=C_L3 Top, parameterName=Add Normals Value, sortPriority=6})
  )
  NamedRerouteDeclaration(None=Normalize(VectorInput=StaticSwitchParameter(True=@n161, False=@n159) {group=C_L3 Top, parameterName=Add Normal to Lower Layer, sortPriority=-2})) {name=CustomTopNormal, nodeColor=<rgba 1 0 0.9141 1>}
  n162 = TextureSampleParameter2D(UVs=@n141) {
   group=C_L3 Top
   parameterName=SpecularMasks_3
   samplerSource=SSM_Wrap_WorldGroupSettings
   samplerType=SAMPLERTYPE_Masks
   sortPriority=1
   texture=&T_Fortnite_Default_S
  }
  n163 = ShadingPathSwitch(Default=StaticSwitch(True=@n146.XYZ Spec, False=@n146.Z Spec, Value=@n143), Mobile=@n162.RGB)
  n164 = MaterialFunctionCall(Float3=StaticSwitchParameter(True=StaticSwitch(True=@n163, False=@n162.RGB, Value=@n139), False=@n163) {group=B_L2, parameterName=Baked Prop UV2, sortPriority=0}) {materialFunction=&BreakOutFloat3Components}
  NamedRerouteDeclaration(None=@n164.B) {name=CustomTopRoughness, nodeColor=<rgba 0 1 0.7656 1>}
  NamedRerouteDeclaration(None=@n164.R) {name=CustomTopSpecular, nodeColor=<rgba 1 0.4453 0 1>}
  NamedRerouteDeclaration(None=ComponentMask(None=@n148) {b=true, g=true, r=true}) {name=CustomTopBaseColor, nodeColor=<rgba 0.125 0 1 1>}
  NamedRerouteUsage()
  OneMinus(None=@n91)
  PinBase() {reroutePins [
    A {expression=&self:MaterialExpressionReroute_221}
    Alpha {expression=&self:MaterialExpressionReroute_222}
   ]}
  PinBase!(Output=@n19) {reroutePins [Output {expression=&self:MaterialExpressionReroute_223}]}
  PinBase!() {reroutePins [
    MaterialAttributes {expression=&self:MaterialExpressionReroute_224}
    Snow Normals {expression=&self:MaterialExpressionReroute_225}
    L1 Normals {expression=&self:MaterialExpressionReroute_226}
   ]}
  MaterialExpressionReroute_225 = Reroute()
  MaterialExpressionReroute_226 = Reroute()
  n165 = Add(A=Multiply(A=@MaterialExpressionReroute_225.Snow Normals) {constB=2}, B=Transform(None=@MaterialExpressionReroute_226.L1 Normals))
  MaterialExpressionReroute_224 = Reroute()
  n166 = StaticSwitch(
   True=SetMaterialAttributes!(MaterialAttributes=@MaterialExpressionReroute_224.MaterialAttributes, Normal=Multiply(A=@n165) {constB=0.5}) {attributeSetTypes [#4]}
   False=@MaterialExpressionReroute_224.MaterialAttributes
   Value=StaticBoolParameter() {group=B_L2, parameterName=Baked Prop UV2, sortPriority=0}
  )
  PinBase!(Output=@n166) {reroutePins [Output {expression=&self:MaterialExpressionReroute_227}]}
  PreSkinnedPosition()
  Reroute(None=SquareRoot(None=@n117.XYZ Diffuse Texture))
  MaterialExpressionReroute_227 = Reroute(None=@n166)
  ScalarParameter() {defaultValue=1, group=B_L2, parameterName=Landscape Noise Amount_Color}
  StaticBoolParameter() {group=B_L2, parameterName=Baked Prop UV2, sortPriority=0}
  n167 = TextureSampleParameter2D(UVs=@n24) {
   group=A_L1_Rock Surface
   parameterName=Diffuse
   samplerSource=SSM_Wrap_WorldGroupSettings
   sortPriority=0
   texture=&T_MidgreysRGB
  }
  StaticSwitchParameter(True=@n167.RGB, False=@n12.XYZ Diffuse Texture) {parameterName=debug}
  n168 = TextureSampleParameter2D(UVs=@n24) {
   group=A_L1_Rock Surface
   parameterName=SpecularMasks
   samplerSource=SSM_Wrap_WorldGroupSettings
   samplerType=SAMPLERTYPE_Masks
   sortPriority=1
   texture=&T_Fortnite_Default_S
  }
  StaticSwitchParameter(True=@n168.RGB) {parameterName=debug}
  StaticSwitchParameter(True=@n167.A, False=@n32.A) {parameterName=debug}
  Subtract(A=WorldPosition().XYZ, B=ObjectPositionWS())
 ]}
