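# Nanite-enabled Default Lit material for rock props in the Helios Central Picnic environment.
# MF_Central_Picnic_HeroRock builds the base material attributes.
# Its normal is converted from world to tangent space and its specular slightly boosted before being written back.
# The result passes through the shared MF_RVTBlendingRocks function, which by its name blends the rock using Runtime Virtual Textures.
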Material /Game/Environments/Helios/Props/CentralPicnic/Materials/M_CentralPicnic_Rock
graph {
 properties {bUseMaterialAttributes=true, bUsedWithNanite=true, pixelDepthOffsetMode=PDOM_Legacy, usedShadingModels=Default Lit}
 nodes [
  n1 = GetMaterialAttributes(None=MaterialFunctionCall() {materialFunction=&/MF_Central_Picnic_HeroRock}.Result) {attributeGetTypes [#1, #2]}
  MaterialFunctionCall!(
   In=SetMaterialAttributes!(
    MaterialAttributes=@n1.MaterialAttributes
    Normal=Transform(None=@n1.Normal) {transformSourceType=TRANSFORMSOURCE_World, transformType=TRANSFORM_Tangent}
    Specular=Multiply(A=@n1.Specular) {constB=1.2}
    Emissive Color
   ) {attributeSetTypes [#1, #2, #3]}
   "RVT Height Map (Roughness)"
  ) {desc="RVT ", materialFunction=/Game/Global/Base4Layers_2022/MaterialFunctions/MF_RVTBlendingRocks}
 ]
}
