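# CentralPicnic variant of the Fortnite 2022 base prop material, supporting Nanite, instanced and spline meshes.
# GigaMix layers receive optional baked AO/dirt, then shared damage, wind, fan, fire and DirectFX functions.
# World-space normals support jungle and RVT blending before faked chest-orb lighting and low-end mobile simplification.
# Static switches finish with either chest AO masking emissive or a crater-dirt overlay, with quality switches cutting dirt costs on Low/Medium.
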Material /Game/Environments/Helios/Props/CentralPicnic/Materials/P_FN_Base_2022_CentralPicnic
graph {
 properties {
  bAutomaticallySetUsageInEditor=false
  bTangentSpaceNormal=false
  bUseMaterialAttributes=true
  bUsedWithInstancedStaticMeshes=true
  bUsedWithNanite=true
  bUsedWithSplineMeshes=true
  ditheredLODTransition=true
  pixelDepthOffsetMode=PDOM_Legacy
  shadingModel=MSM_FromMaterialExpression
  usedShadingModels=Default Lit | Subsurface | Clear Coat | Two Sided Foliage
 }
 nodes [
  n1 = MaterialFunctionCall() {materialFunction=/Game/Global/Base4Layers_2022/MaterialFunctions/MF_GigaMix_2023}
  n2 = TextureSampleParameter2D(UVs=TextureCoordinate() {coordinateIndex=2}) {group=Baked_AO, parameterName=Baked Dirt/AO, texture=/Game/Environments/Helios/Props/CentralPicnic/Textures/T_CentralPicnic_Box_Dirt_Base_D}
  n3 = StaticSwitchParameter(
   True=MaterialFunctionCall!(Material=@n1.Result, NewBaseColor=@n2.A, Mask=ScalarParameter() {defaultValue=1, group=Baked_AO, parameterName=AO amount, sliderMax=1}) {materialFunction=/Engine/Functions/MaterialLayerFunctions/MatLayerBlend_MultiplyBaseColor}.Blended Material
   False=@n1.Result
  ) {group=Baked_AO, parameterName=Use Baked AO}
  n4 = ScalarParameter() {defaultValue=0.5, group=Baked_AO, parameterName=Baked Dirt Opacity}
  n5 = GetMaterialAttributes(None=MaterialFunctionCall!(Material=@n3, Tint=@n2.RGB, TintMask=@n4) {materialFunction=/Engine/Functions/MaterialLayerFunctions/MatLayerBlend_Tint}.Blended Material) {attributeGetTypes [#1, #2, #3]}
  n6 = Add(A=OneMinus(None=@n2.RGB), B=@n5.Roughness)
  n7 = Multiply(A=@n2.R, B=@n5.Metallic)
  n8 = Multiply(A=@n2.R, B=@n5.Specular)
  n9 = StaticSwitchParameter(
   True=SetMaterialAttributes!(
    MaterialAttributes=@n5.MaterialAttributes
    Roughness=LinearInterpolate(A=@n5.Roughness, B=@n6, Alpha=@n4)
    Metallic=LinearInterpolate(A=@n5.Metallic, B=@n7, Alpha=@n4)
    Specular=LinearInterpolate(A=@n5.Specular, B=@n8, Alpha=@n4)
   ) {attributeSetTypes [#3, #1, #2]}
   False=@n3
  ) {group=Baked_AO, parameterName=Use Baked Dirt}
  n10 = QualitySwitch(Default=@n9, Low=@n1.Result, High=@n9, Medium=@n1.Result, Epic=@n9)
  MaterialExpressionReroute_70 = Reroute(None=@n10)
  n11 = ScalarParameter() {
   desc=Documentation link
   group="0 (Expand)  Documentation (RMB - Copy display name) and Debug view"
   parameterName=https://confluence.it.epicgames.com/display/FORT/Fortnite+Base+Material+2022
  }
  n12 = SetMaterialAttributes!(MaterialAttributes=@MaterialExpressionReroute_70.MaterialAttributes, World Position Offset=Multiply(A=@n11)) {attributeSetTypes [#4]}
  MaterialExpressionReroute_71 = Reroute(None=@n12)
  MaterialExpressionReroute_69 = Reroute(None=@MaterialExpressionReroute_71.Output)
  n13 = MaterialFunctionCall(Material_In=@MaterialExpressionReroute_69.MaterialAttributes) {materialFunction=&MF_FN_DestructibleDamage}
  n14 = MaterialFunctionCall!(
   In=MaterialFunctionCall!(Material Attributes=MaterialFunctionCall!(In=@n13.Material_Out) {materialFunction=/Game/Athena/WinterArt/Materials/MF_WPO_WindNew}.Result) {materialFunction=/Game/Creative/Sets/MilitaryBase/Props/MF_FanSpin}.Result
   UseSphericalMappingOverride
  ) {materialFunction=/Game/Packages/Fortress_SharedMaterials/Base_Material_Functions/MF_CurieFallbackEffect}
  MaterialExpressionReroute_68 = Reroute(None=@n14.Result)
  Composite!(None=@MaterialExpressionReroute_68.Result) {desc=Transform to World in case we will use RVT after this point, subgraphName=Tangent to World}
  Composite!(MaterialAttributes=@n10) {subgraphName=Documentation Link}
  Composite(MaterialAttributes=@MaterialExpressionReroute_71.Output) {desc=Game Shared WPO and Fire Effect, subgraphName=WPO And Shared Functions}
  MaterialExpressionReroute_1 = Reroute(None=@n19.RVT H)
  MaterialExpressionReroute_3 = Reroute(None=@MaterialExpressionReroute_68.Result)
  n15 = GetMaterialAttributes(None=@MaterialExpressionReroute_3.Input) {attributeGetTypes [#5]}
  n16 = SetMaterialAttributes!(MaterialAttributes=@n15.MaterialAttributes, Normal=Transform(None=@n15.Normal)) {attributeSetTypes [#5]}
  MaterialExpressionReroute_2 = Reroute(None=@n16)
  n17 = MaterialFunctionCall!(In=@MaterialExpressionReroute_2.Output) {materialFunction=/Game/ArtTools/Tech_Art/DirectFX/Assets/MaterialFunctions/MF_DirectFX}
  MaterialExpressionReroute_8 = Reroute(None=@n17.Result)
  n18 = MaterialFunctionCall!(AO, RVT_Height=@MaterialExpressionReroute_1."RVT_Height (S)", For Buildings=@MaterialExpressionReroute_8."For Moss (MA)", For Moss) {materialFunction=/Game/Environments/Asteria/Rocks/Materials/MaterialFunctions/MF_JungleBiome}
  MaterialExpressionReroute_0 = Reroute(None=@n18.Buildings Jungl)
  n19 = MaterialFunctionCall!(
   In=@MaterialExpressionReroute_0.Result
   "RVT Height Map (Roughness)"=GetMaterialAttributes(None=@n17.Result) {attributeGetTypes [#3]}.Roughness
  ) {materialFunction=/Game/Global/Base4Layers_2022/MaterialFunctions/MF_RVTBlending}
  n20 = GetMaterialAttributes(None=@n19.Result) {attributeGetTypes [#5]}
  n21 = MaterialFunctionCall!(Material Attributes=@n20.MaterialAttributes, World Normal Map=@n20.MaterialAttributes, Vertex Normals, Use Vertex Normal Masking) {materialFunction=&/MF_CentralPicnic_ChestOrbFakedLight}
  Composite!(InputPin=@n21.Material) {subgraphName=Mobile Opt}
  Composite!("For Moss (MA)"=@n17.Result, "RVT_Height (S)"=@n19.RVT H) {subgraphName="Seasonal "}
  PinBase() {reroutePins [Input {expression=&self:MaterialExpressionReroute_3}]}
  PinBase!(Output=@n16) {reroutePins [Output {expression=&self:MaterialExpressionReroute_2}]}
  PinBase!() {reroutePins [
    "For Moss (MA)" {expression=&self:MaterialExpressionReroute_8}
    "RVT_Height (S)" {expression=&self:MaterialExpressionReroute_1}
   ]}
  PinBase!(Result=@n18.Buildings Jungl) {reroutePins [Result {expression=&self:MaterialExpressionReroute_0}]}
  PinBase!() {reroutePins [MaterialAttributes {expression=&self:MaterialExpressionReroute_70}]}
  PinBase!(Output=@n12) {reroutePins [Output {expression=&self:MaterialExpressionReroute_71}]}
  PinBase!() {reroutePins [MaterialAttributes {expression=&self:MaterialExpressionReroute_69}]}
  PinBase!(Result=@n14.Result) {reroutePins [Result {expression=&self:MaterialExpressionReroute_68}]}
  PinBase!() {reroutePins [InputPin {expression=&self:MaterialExpressionReroute_7}]}
  MaterialExpressionReroute_7 = Reroute(None=@n21.Material)
  n22 = SetMaterialAttributes!(
   MaterialAttributes=GetMaterialAttributes(None=@MaterialExpressionReroute_7.InputPin).MaterialAttributes
   Metallic=Constant() {r=1}
   Specular=Constant() {r=0.55}
   Roughness=Constant() {r=0.65}
  ) {attributeSetTypes [#1, #2, #3]}
  n23 = ShadingPathSwitch(
   Default=@MaterialExpressionReroute_7.InputPin
   Mobile=QualitySwitch(
    Default=@n22
    Low=@n22
    High=@MaterialExpressionReroute_7.InputPin
    Medium=@MaterialExpressionReroute_7.InputPin
    Epic=@MaterialExpressionReroute_7.InputPin
   )
  )
  PinBase!(Output=@n23) {reroutePins [Output {expression=&self:MaterialExpressionReroute_6}]}
  MaterialExpressionReroute_6 = Reroute(None=@n23)
  n24 = MaterialFunctionCall(MA In=@MaterialExpressionReroute_6.Output) {materialFunction=&MF_CreativeMoveTool}
  n25 = TextureSampleParameter2D() {
   constCoordinate=2
   group=Texture Overlay
   parameterName=Texture Overlay Normal
   samplerType=SAMPLERTYPE_Normal
   texture=/Game/Environments/Helios/Props/CentralPicnic/Textures/T_CentralPicnic_Box_CraterDirt_Base_N
  }
  n26 = TextureSampleParameter2D() {
   constCoordinate=2
   group=Texture Overlay
   parameterName=Texture Overlay Diffuse
   texture=/Game/Environments/Helios/Props/CentralPicnic/Textures/T_CentralPicnic_Box_CraterDirt_Base_D
  }
  n27 = GetMaterialAttributes(None=MaterialFunctionCall!(Material=@n24.MA Out, Additional Normal=@n25.RGB, Mask=@n26.A) {materialFunction=/Engine/Functions/MaterialLayerFunctions/MatLayerBlend_BlendAngleCorrectedNormals}.Blended Material) {attributeGetTypes [
    #6
    #1
    #2
    #3
   ]}
  n28 = LinearInterpolate(A=@n27.Base Color, B=@n26.RGB, Alpha=@n26.RGB)
  n29 = TextureSampleParameter2D() {
   constCoordinate=2
   group=Texture Overlay
   parameterName=Texture Overlay Specular
   samplerType=SAMPLERTYPE_Masks
   texture=/Game/Environments/Helios/Props/CentralPicnic/Textures/T_CentralPicnic_Box_CraterDirt_Base_S
  }
  n30 = SetMaterialAttributes!(
   MaterialAttributes=@n27.MaterialAttributes
   Base Color=LinearInterpolate(A=@n28, B=Multiply(A=@n2.R, B=@n28)) {constAlpha=0.65}
   Metallic=LinearInterpolate(A=@n27.Metallic, B=@n29.G, Alpha=@n26.A)
   Specular=LinearInterpolate(A=@n27.Specular, B=@n29.R, Alpha=@n26.A)
   Roughness=LinearInterpolate(A=@n27.Roughness, B=@n29.B, Alpha=@n26.A)
  ) {attributeSetTypes [
    #6
    #1
    #2
    #3
   ]}
  n31 = GetMaterialAttributes(None=@n24.MA Out) {attributeGetTypes [
    #6
    #1
    #2
    #3
   ]}
  n32 = SetMaterialAttributes(
   MaterialAttributes=@n31.MaterialAttributes
   Base Color=LinearInterpolate(A=@n31.Base Color, B=@n26.RGB, Alpha=@n26.RGB)
   Metallic=LinearInterpolate(A=@n31.Metallic, B=Constant(), Alpha=@n26.A)
   Specular=LinearInterpolate(A=@n31.Specular, B=Constant() {r=0.035}, Alpha=@n26.A)
   Roughness=LinearInterpolate(A=@n31.Roughness, B=Constant() {r=0.85}, Alpha=@n26.A)
  ) {attributeSetTypes [
    #6
    #1
    #2
    #3
   ]}
  n33 = GetMaterialAttributes(None=@n24.MA Out) {attributeGetTypes [#7]}
  StaticSwitchParameter(
   True=SetMaterialAttributes!(MaterialAttributes=@n33.MaterialAttributes, Emissive Color=Multiply(A=@n33.Emissive Color, B=@n2.A)) {attributeSetTypes [#7]}
   False=StaticSwitchParameter(True=QualitySwitch(Default=@n30, Low=@n32, High=@n30, Medium=@n32, Epic=@n30), False=@n24.MA Out) {group=Texture Overlay, parameterName=Textured Dirt}
  ) {group=Baked_AO, parameterName=Chest_AO}
 ]
}
